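# Physics asset for the cape of the male medium InspireSpell_Alt player body.
# Six front, side and back chains, left and right, each run three capsules from a zero-size base, with limited swing and very narrow twist.
# Free-rotation constraints cross-link some neighboring tips: left side to front and back, right side to back only.
# Thigh, calf and foot capsules plus a huge floor_joint capsule provide leg and ground shapes.
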PhysicsAsset /BRCosmetics/Characters/Player/Male/Medium/Bodies/M_MED_InspireSpell_Alt/Meshes/M_MED_InspireSpell_Alt_Cape_Physics
bodies {
 calf_l [{
   shapeType=Capsule
   length=40.25
   radius=8.5
   shapeName=thigh_r_capsule
   center=<-12.4936 -1.1814 0.5625>
   rotation=<rot 88.1065 10.4919 12.5739>
  }]
 calf_r [{
   shapeType=Capsule
   length=41.5
   radius=7.25
   shapeName=thigh_r_capsule
   center=<12.5358 2.7614 -0.4241>
   rotation=<rot 88.1065 10.4919 12.574>
  }]
 dyn_cape_b_l_01 [{shapeType=Capsule, length=13, radius=12, shapeName=dyn_cape_b_l_01_capsule, center=<0 0 8>, rotation=<rot 0 0 90>}]
 dyn_cape_b_l_02 [{shapeType=Capsule, length=15, radius=10, shapeName=dyn_cape_b_l_02_capsule, center=<8.9173 0 6>, rotation=<rot 0 0 90>}]
 dyn_cape_b_l_03 [{shapeType=Capsule, length=17.5, radius=7.5, shapeName=dyn_cape_b_l_03_capsule, center=<11.2353 0 3>, rotation=<rot 0 0 90>}]
 dyn_cape_b_l_base [{shapeType=Capsule, length=0, radius=0, shapeName=dyn_cape_b_l_base_capsule}]
 dyn_cape_b_r_01 [{shapeType=Capsule, length=13, radius=12, shapeName=dyn_cape_b_r_01_capsule, center=<0 0 8>, rotation=<rot 0 0 90>}]
 dyn_cape_b_r_02 [{shapeType=Capsule, length=15, radius=10, shapeName=dyn_cape_b_r_02_capsule, center=<8.9173 0 6>, rotation=<rot 0 0 90>}]
 dyn_cape_b_r_03 [{shapeType=Capsule, length=24.25, radius=7.5, shapeName=dyn_cape_b_r_03_capsule, center=<11.2353 0 3>, rotation=<rot 0 0 90>}]
 dyn_cape_b_r_base [{shapeType=Capsule, length=0, radius=0, shapeName=dyn_cape_b_r_base_capsule}]
 dyn_cape_f_l_01 [{shapeType=Capsule, length=8.75, radius=10.75, shapeName=dyn_cape_f_l_01_capsule, center=<0 0 9>, rotation=<rot 0 0 90>}]
 dyn_cape_f_l_02 [{shapeType=Capsule, length=15, radius=10, shapeName=dyn_cape_f_l_02_capsule, center=<8.9173 0 7>, rotation=<rot 0 0 90>}]
 dyn_cape_f_l_03 [{shapeType=Capsule, length=25, radius=7.5, shapeName=dyn_cape_f_l_03_capsule, center=<11.2353 0 4.5>, rotation=<rot 0 0 90>}]
 dyn_cape_f_l_base [{shapeType=Capsule, length=0, radius=0, shapeName=dyn_cape_f_l_base_capsule}]
 dyn_cape_f_r_01 [{shapeType=Capsule, length=13, radius=12, shapeName=dyn_cape_f_r_01_capsule, center=<0 0 9>, rotation=<rot 0 0 90>}]
 dyn_cape_f_r_02 [{shapeType=Capsule, length=15, radius=10, shapeName=dyn_cape_f_r_02_capsule, center=<7.2299 0 7>, rotation=<rot 0 0 90>}]
 dyn_cape_f_r_03 [{shapeType=Capsule, length=17.5, radius=7.5, shapeName=dyn_cape_f_r_03_capsule, center=<11.2353 0 4.5>, rotation=<rot 0 0 90>}]
 dyn_cape_f_r_base [{shapeType=Capsule, length=0, radius=0, shapeName=dyn_cape_f_r_base_capsule}]
 dyn_cape_s_l_01 [{shapeType=Capsule, length=13, radius=12, shapeName=dyn_cape_s_l_01_capsule, center=<0 0 13.5>, rotation=<rot 0 0 90>}]
 dyn_cape_s_l_02 [{shapeType=Capsule, length=15, radius=10, shapeName=dyn_cape_s_l_02_capsule, center=<8.9173 0 8>, rotation=<rot 0 0 90>}]
 dyn_cape_s_l_03 [{shapeType=Capsule, length=26.75, radius=7.5, shapeName=dyn_cape_s_l_03_capsule, center=<11.2353 0 3>, rotation=<rot 0 0 90>}]
 dyn_cape_s_l_base [{shapeType=Capsule, length=0, radius=0, shapeName=dyn_cape_s_l_base_capsule}]
 dyn_cape_s_r_01 [{shapeType=Capsule, length=13, radius=12, shapeName=dyn_cape_s_r_01_capsule, center=<0 0 13.5>, rotation=<rot 0 0 90>}]
 dyn_cape_s_r_02 [{shapeType=Capsule, length=15, radius=10, shapeName=dyn_cape_s_r_02_capsule, center=<10.7069 0.0159 6.5>, rotation=<rot 0 0 90>}]
 dyn_cape_s_r_03 [{shapeType=Capsule, length=28, radius=7.5, shapeName=dyn_cape_s_r_03_capsule, center=<11.2353 0 6>, rotation=<rot 0 0 90>}]
 dyn_cape_s_r_base [{shapeType=Capsule, length=0, radius=0, shapeName=dyn_cape_s_r_base_capsule}]
 floor_joint [{shapeType=Capsule, length=48.3294, radius=165.0692, shapeName=root_capsule, center=<2.6954 -23.3737 -154.0549>, rotation=<rot 0 0 -90>}]
 foot_l [{shapeType=Capsule, length=15, radius=12.75, shapeName=foot_l_capsule, center=<-4.0759 0.0432 -0.0194>, rotation=<rot 0 0 90>}]
 foot_r [{shapeType=Capsule, length=15, radius=12.25, shapeName=foot_r_capsule, center=<4.0759 -0.0432 0.0194>, rotation=<rot 0 0 90>}]
 thigh_l [{
   shapeType=Capsule
   length=37.25
   radius=13
   shapeName=thigh_r_capsule
   center=<-16.7501 -2.0676 -1.2562>
   rotation=<rot 85.3118 179.8759 176.9445>
  }]
 thigh_r [{
   shapeType=Capsule
   length=38.25
   radius=13
   shapeName=thigh_r_capsule
   center=<15.9216 1.0764 0.0566>
   rotation=<rot 86.7774 178.3456 -179.5829>
  }]
}
constraints [
 {
  bone1Name=calf_l
  bone2Name=thigh_l
  swing1LimitDegrees=45
  swing1Motion=Locked
  swing2LimitDegrees=45
  swing2Motion=Locked
  twistLimitDegrees=45
  twistMotion=Locked
 }
 {
  bone1Name=calf_r
  bone2Name=thigh_r
  swing1LimitDegrees=45
  swing1Motion=Locked
  swing2LimitDegrees=45
  swing2Motion=Locked
  twistLimitDegrees=45
  twistMotion=Locked
 }
 {bone1Name=dyn_cape_b_l_01, bone2Name=dyn_cape_b_l_base, swing1LimitDegrees=9.84, swing2LimitDegrees=35, twistLimitDegrees=2}
 {bone1Name=dyn_cape_b_l_02, bone2Name=dyn_cape_b_l_01, swing1LimitDegrees=10, swing2LimitDegrees=35, twistLimitDegrees=2}
 {bone1Name=dyn_cape_b_l_03, bone2Name=dyn_cape_b_l_02, swing1LimitDegrees=5, swing2LimitDegrees=35, twistLimitDegrees=2}
 {bone1Name=dyn_cape_b_r_01, bone2Name=dyn_cape_b_r_base, swing1LimitDegrees=12.72, swing2LimitDegrees=35, twistLimitDegrees=2}
 {bone1Name=dyn_cape_b_r_02, bone2Name=dyn_cape_b_r_01, swing1LimitDegrees=10, swing2LimitDegrees=35, twistLimitDegrees=2}
 {bone1Name=dyn_cape_b_r_03, bone2Name=dyn_cape_b_r_02, swing1LimitDegrees=5, swing2LimitDegrees=35, twistLimitDegrees=2}
 {bone1Name=dyn_cape_f_l_01, bone2Name=dyn_cape_f_l_base, swing1LimitDegrees=30, swing2LimitDegrees=40, twistLimitDegrees=2}
 {bone1Name=dyn_cape_f_l_02, bone2Name=dyn_cape_f_l_01, swing1LimitDegrees=30, swing2LimitDegrees=35, twistLimitDegrees=2}
 {bone1Name=dyn_cape_f_l_03, bone2Name=dyn_cape_f_l_02, swing1LimitDegrees=60, swing2LimitDegrees=35, twistLimitDegrees=2}
 {bone1Name=dyn_cape_f_r_01, bone2Name=dyn_cape_f_r_base, swing1LimitDegrees=25, swing2LimitDegrees=30, twistLimitDegrees=2}
 {bone1Name=dyn_cape_f_r_02, bone2Name=dyn_cape_f_r_01, swing1LimitDegrees=30, swing2LimitDegrees=35, twistLimitDegrees=2}
 {bone1Name=dyn_cape_f_r_03, bone2Name=dyn_cape_f_r_02, swing1LimitDegrees=30, swing2LimitDegrees=35, twistLimitDegrees=2}
 {bone1Name=dyn_cape_s_l_01, bone2Name=dyn_cape_s_l_base, swing1LimitDegrees=20, swing2LimitDegrees=35, twistLimitDegrees=2}
 {bone1Name=dyn_cape_s_l_02, bone2Name=dyn_cape_s_l_01, swing1LimitDegrees=12.96, swing2LimitDegrees=35, twistLimitDegrees=2}
 {bone1Name=dyn_cape_s_l_03, bone2Name=dyn_cape_s_l_02, swing1LimitDegrees=12.72, swing2LimitDegrees=35, twistLimitDegrees=2}
 {bone1Name=dyn_cape_s_r_01, bone2Name=dyn_cape_s_r_base, swing1LimitDegrees=20, swing2LimitDegrees=25, twistLimitDegrees=2}
 {bone1Name=dyn_cape_s_r_02, bone2Name=dyn_cape_s_r_01, swing1LimitDegrees=17.04, swing2LimitDegrees=25, twistLimitDegrees=2}
 {bone1Name=dyn_cape_s_r_03, bone2Name=dyn_cape_s_r_02, swing1LimitDegrees=30, swing2LimitDegrees=35, twistLimitDegrees=2}
 {
  bone1Name=foot_r
  bone2Name=calf_r
  swing1LimitDegrees=45
  swing1Motion=Free
  swing2LimitDegrees=45
  swing2Motion=Free
  twistLimitDegrees=45
  twistMotion=Free
 }
 {
  bone1Name=foot_l
  bone2Name=calf_l
  swing1LimitDegrees=45
  swing1Motion=Free
  swing2LimitDegrees=45
  swing2Motion=Free
  twistLimitDegrees=45
  twistMotion=Free
 }
 {
  bone1Name=dyn_cape_s_r_03
  bone2Name=dyn_cape_b_r_03
  swing1LimitDegrees=45
  swing1Motion=Free
  swing2LimitDegrees=45
  swing2Motion=Free
  twistLimitDegrees=45
  twistMotion=Free
 }
 {
  bone1Name=dyn_cape_s_l_03
  bone2Name=dyn_cape_f_l_03
  swing1LimitDegrees=45
  swing1Motion=Free
  swing2LimitDegrees=45
  swing2Motion=Free
  twistLimitDegrees=45
  twistMotion=Free
 }
 {
  bone1Name=dyn_cape_b_l_03
  bone2Name=dyn_cape_s_l_03
  swing1LimitDegrees=45
  swing1Motion=Free
  swing2LimitDegrees=45
  swing2Motion=Free
  twistLimitDegrees=45
  twistMotion=Free
 }
]
