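# Physics asset for the hood of the M_MED_InspireSpell_Alt male medium player body cosmetic, defining collision and joints for its dynamic hood bones.
# Three capsule bodies form a chain from a small dyn_hood_base anchor through dyn_hood_01 to dyn_hood_02.
# Two constraints link neighboring bodies with zero twist and swing1 limit angles and a small swing2 angle, tighter at the tip, suggesting a mostly single-axis flap.
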PhysicsAsset /BRCosmetics/Characters/Player/Male/Medium/Bodies/M_MED_InspireSpell_Alt/Meshes/M_MED_InspireSpell_Alt_Hood_Physics
bodies {
 dyn_hood_01 [{shapeType=Capsule, length=4.3743, radius=2.8669, shapeName=dyn_hood_01_capsule, center=<0 -0.5975 0>, rotation=<rot 90 -171.1695 -161.1698>}]
 dyn_hood_02 [{shapeType=Capsule, length=3.0947, radius=2.655, shapeName=dyn_hood_02_capsule, center=<2.158 -1.713 0>, rotation=<rot 85 -179.9998 170>}]
 dyn_hood_base [{shapeType=Capsule, length=0.505, radius=0.505, shapeName=dyn_hood_base_capsule, rotation=<rot 0 -90 0>}]
}
constraints [
 {bone1Name=dyn_hood_01, bone2Name=dyn_hood_base, swing1LimitDegrees=0, swing2LimitDegrees=20, twistLimitDegrees=0}
 {bone1Name=dyn_hood_02, bone2Name=dyn_hood_01, swing1LimitDegrees=0, swing2LimitDegrees=10, twistLimitDegrees=0}
]
