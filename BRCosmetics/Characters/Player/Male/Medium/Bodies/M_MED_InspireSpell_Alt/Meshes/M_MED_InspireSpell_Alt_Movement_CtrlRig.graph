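# Control Rig for the M_MED_InspireSpell_Alt player body that procedurally corrects cape bones as the body moves.
# With rigid body enabled, spherical pose readers on the thighs and spine_04 weight additive offsets to cape base and front corrective bones.
# In FrontEnd, fixed front-cape offsets replace one left-thigh reader.
# With rigid body disabled, it instead shifts altskin_cape_back by scaled calf-to-calf distance, measured through constrained nulls.
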ControlRigBlueprint /BRCosmetics/Characters/Player/Male/Medium/Bodies/M_MED_InspireSpell_Alt/Meshes/M_MED_InspireSpell_Alt_Movement_CtrlRig
graphs [
 ContainedGraph {nodes [
   n1 = RigVMFunctionEntryNode()
   RigUnit_SequenceAggregate = &RigVMFunction_Sequence(ExecuteContext=@n1.ExecuteContext)
   RigUnit_SequenceAggregate_1 = &RigVMFunction_Sequence(ExecuteContext=@RigUnit_SequenceAggregate.B)
   RigUnit_SequenceAggregate_2 = &RigVMFunction_Sequence(ExecuteContext=@RigUnit_SequenceAggregate_1.B)
   RigUnit_SequenceAggregate_3 = &RigVMFunction_Sequence(ExecuteContext=@RigUnit_SequenceAggregate_2.B)
   RigUnit_SequenceAggregate_4 = &RigVMFunction_Sequence(ExecuteContext=@RigUnit_SequenceAggregate_3.B)
   RigVMFunctionReturnNode(
    A=@RigUnit_SequenceAggregate.A
    B=@RigUnit_SequenceAggregate_1.A
    C=@RigUnit_SequenceAggregate_2.A
    D=@RigUnit_SequenceAggregate_3.A
    E=@RigUnit_SequenceAggregate_4.A
    F=@RigUnit_SequenceAggregate_4.B
   )
  ]}
 RigVMFunctionLibrary {links [], nodes []}
 RigVMModel {nodes [
   RigUnit_BeginExecution = &RigUnit_BeginExecution()
   BranchNode_RigVMUnitNode = &RigVMFunction_ControlFlowBranch(
    Condition=&RigVMFunction_MathBoolNot(Value=RigVMVariableNode() {variable=Is enabled rigid body}.Value).Result
    ExecuteContext=@RigUnit_BeginExecution.ExecutePin
   )
   RigUnit_SequenceAggregate = RigVMAggregateNode(ExecuteContext=@BranchNode_RigVMUnitNode.False)
   n1 = &RigUnit_SphericalPoseReader(
    ActiveRegionScaleFactors=(PositiveWidth=1,NegativeWidth=1,PositiveHeight=1,NegativeHeight=1)
    ActiveRegionSize=0.1
    Debug=(bDrawDebug=false,DebugScale=25,DebugSegments=20,DebugThickness=0.25)
    DriverAxis=(X=1,Y=0,Z=0)
    DriverItem=(Type=Bone,Name="thigh_r")
    FalloffRegionScaleFactors=(PositiveWidth=1,NegativeWidth=1,PositiveHeight=1,NegativeHeight=1)
    FalloffSize=0.2
    OptionalParentItem=(Type=Bone)
    RotationOffset=(X=-20,Y=-50,Z=0)
    ExecutePin=@RigUnit_SequenceAggregate.B
   )
   > &RigUnit_ModifyTransforms(
    ItemToModify=((Item=(Type=Bone,Name="dyn_cape_f_r_base"),Transform=(Rotation=(X=0,Y=0,Z=0,W=1),Translation=(X=0,Y=0,Z=0),Scale3D=(X=1,Y=1,Z=1))))
    ItemToModify.0.Transform.Rotation=&RigVMFunction_MathQuaternionFromEuler(Euler=(X=0,Y=-30,Z=0)).Result
    Weight=@n1.OutputParam
   )
   > &RigUnit_ModifyTransforms(
    ItemToModify=((Item=(Type=Bone,Name="dyn_cape_s_r_base"),Transform=(Rotation=(X=0,Y=0,Z=0,W=1),Translation=(X=0,Y=0,Z=0),Scale3D=(X=1,Y=1,Z=1))))
    ItemToModify.0.Transform.Rotation=&RigVMFunction_MathQuaternionFromEuler(Euler=(X=0,Y=-30,Z=0)).Result
    Weight=@n1.OutputParam
   )
   BranchNode_1_RigVMUnitNode = &RigVMFunction_ControlFlowBranch(ExecuteContext=@RigUnit_SequenceAggregate.C, Condition=RigVMVariableNode() {variable=Is FrontEnd}.Value)
   n2 = &RigUnit_SphericalPoseReader(
    ActiveRegionScaleFactors=(PositiveWidth=1,NegativeWidth=1,PositiveHeight=1,NegativeHeight=1)
    ActiveRegionSize=0.1
    Debug=(bDrawDebug=false,DebugScale=25,DebugSegments=20,DebugThickness=0.25)
    DriverAxis=(X=-1,Y=0,Z=0)
    DriverItem=(Type=Bone,Name="thigh_l")
    FalloffRegionScaleFactors=(PositiveWidth=1,NegativeWidth=1,PositiveHeight=1,NegativeHeight=1)
    FalloffSize=0.2
    OptionalParentItem=(Type=Bone)
    RotationOffset=(X=20,Y=130,Z=0)
    ExecutePin=@BranchNode_1_RigVMUnitNode.False
   )
   > &RigUnit_ModifyTransforms(
    ItemToModify=((Item=(Type=Bone,Name="dyn_cape_f_l_base"),Transform=(Rotation=(X=0,Y=0,Z=0,W=1),Translation=(X=0,Y=0,Z=0),Scale3D=(X=1,Y=1,Z=1))))
    ItemToModify.0.Transform.Rotation=&RigVMFunction_MathQuaternionFromEuler(Euler=(X=0,Y=-35,Z=0)).Result
    Weight=@n2.OutputParam
   )
   > &RigUnit_ModifyTransforms(
    ItemToModify=((Item=(Type=Bone,Name="dyn_cape_s_l_base"),Transform=(Rotation=(X=0,Y=0,Z=0,W=1),Translation=(X=0,Y=0,Z=0),Scale3D=(X=1,Y=1,Z=1))))
    ItemToModify.0.Transform.Rotation=&RigVMFunction_MathQuaternionFromEuler(Euler=(X=0,Y=-30,Z=0)).Result
    Weight=@n2.OutputParam
   )
   &RigUnit_ModifyTransforms(
    ItemToModify=((Item=(Type=Bone,Name="dyn_cape_f_l_base"),Transform=(Rotation=(X=0,Y=0,Z=0,W=1),Translation=(X=0,Y=0,Z=0),Scale3D=(X=1,Y=1,Z=1))))
    Weight=1
    ExecutePin=@BranchNode_1_RigVMUnitNode.True
    ItemToModify.0.Transform.Rotation=&RigVMFunction_MathQuaternionFromEuler(Euler=(X=0,Y=-45,Z=0)).Result
   )
   > &RigUnit_ModifyTransforms(
    ItemToModify=((Item=(Type=Bone,Name="dyn_cape_f_r_base"),Transform=(Rotation=(X=0,Y=0,Z=0,W=1),Translation=(X=5,Y=0,Z=0),Scale3D=(X=1,Y=1,Z=1))))
    Weight=1
    ItemToModify.0.Transform.Rotation=&RigVMFunction_MathQuaternionFromEuler(Euler=(X=0,Y=-45,Z=0)).Result
   )
   n3 = &RigUnit_SphericalPoseReader(
    ActiveRegionScaleFactors=(PositiveWidth=1,NegativeWidth=0.3,PositiveHeight=1,NegativeHeight=1)
    ActiveRegionSize=0.3
    Debug=(bDrawDebug=false,DebugScale=25,DebugSegments=20,DebugThickness=0.25)
    DriverAxis=(X=-1,Y=0,Z=0)
    DriverItem=(Type=Bone,Name="thigh_l")
    FalloffRegionScaleFactors=(PositiveWidth=1,NegativeWidth=1,PositiveHeight=1,NegativeHeight=1)
    FalloffSize=0.5
    OptionalParentItem=(Type=Bone,Name="spine_02")
    RotationOffset=(X=115,Y=0,Z=-25)
    ExecutePin=@RigUnit_SequenceAggregate.D
   )
   > &RigUnit_ModifyTransforms(
    ItemToModify=((Item=(Type=Bone,Name="dyn_cape_f_l_base"),Transform=(Rotation=(X=0,Y=0,Z=0,W=1),Translation=(X=0,Y=-6,Z=5),Scale3D=(X=1,Y=1,Z=1))))
    ItemToModify.0.Transform=/Script/RigVM.RigVMFunction_MathTransformFromSRT(Location=(X=5,Y=-3,Z=5), Rotation=(X=70,Y=-40,Z=0), RotationOrder=XYZ, Scale=(X=1,Y=1,Z=1)) {is=struct RigVMUnitNode}.Transform
    Weight=@n3.OutputParam
   )
   > &RigUnit_ModifyTransforms(
    ItemToModify=((Item=(Type=Bone,Name="dyn_cape_corrective_f_l"),Transform=(Rotation=(X=0,Y=0,Z=0,W=1),Translation=(X=5,Y=-1,Z=2),Scale3D=(X=1,Y=1,Z=1))))
    ItemToModify.0.Transform=&RigVMFunction_MathTransformFromSRT(Location=(X=-2,Y=-4,Z=2), Rotation=(X=0,Y=0,Z=-50), RotationOrder=XYZ, Scale=(X=1,Y=1,Z=1)).Transform
    Weight=@n3.OutputParam
   )
   n4 = &RigUnit_SphericalPoseReader(
    ActiveRegionScaleFactors=(PositiveWidth=1,NegativeWidth=0.3,PositiveHeight=1,NegativeHeight=1)
    ActiveRegionSize=0.3
    Debug=(bDrawDebug=false,DebugScale=25,DebugSegments=20,DebugThickness=0.25)
    DriverAxis=(X=1,Y=0,Z=0)
    DriverItem=(Type=Bone,Name="thigh_r")
    FalloffRegionScaleFactors=(PositiveWidth=1,NegativeWidth=1,PositiveHeight=1,NegativeHeight=1)
    FalloffSize=0.5
    OptionalParentItem=(Type=Bone,Name="spine_02")
    RotationOffset=(X=-90,Y=180,Z=-25)
    ExecutePin=@RigUnit_SequenceAggregate.E
   )
   > &RigUnit_ModifyTransforms(
    ItemToModify=((Item=(Type=Bone,Name="dyn_cape_f_r_base"),Transform=(Rotation=(X=0,Y=0,Z=0,W=1),Translation=(X=0,Y=6,Z=5),Scale3D=(X=1,Y=1,Z=1))))
    ItemToModify.0.Transform=&RigVMFunction_MathTransformFromSRT(Location=(X=5,Y=-3,Z=6.5), Rotation=(X=-70,Y=-40,Z=0), RotationOrder=XYZ, Scale=(X=1,Y=1,Z=1)).Transform
    Weight=@n4.OutputParam
   )
   > &RigUnit_ModifyTransforms(
    ItemToModify=((Item=(Type=Bone,Name="dyn_cape_corrective_f_r"),Transform=(Rotation=(X=0,Y=0,Z=0,W=1),Translation=(X=5,Y=-1,Z=2),Scale3D=(X=1,Y=1,Z=1))))
    ItemToModify.0.Transform=&RigVMFunction_MathTransformFromSRT(Location=(X=2,Y=4,Z=2), Rotation=(X=0,Y=0,Z=50), Scale=(X=1,Y=1,Z=1)).Transform
    Weight=@n4.OutputParam
   )
   > &RigUnit_ModifyTransforms(
    ItemToModify=((Item=(Type=Bone,Name="dyn_cape_s_r_base"),Transform=(Rotation=(X=0,Y=0,Z=0,W=1),Translation=(X=0,Y=1,Z=5),Scale3D=(X=1,Y=1,Z=1))))
    Weight=@n4.OutputParam
   )
   n5 = &RigUnit_SphericalPoseReader(
    ActiveRegionScaleFactors=(PositiveWidth=1,NegativeWidth=1,PositiveHeight=1,NegativeHeight=1)
    ActiveRegionSize=0.07
    Debug=(bDrawDebug=false,DebugScale=25,DebugSegments=20,DebugThickness=0.25)
    DriverAxis=(X=-1,Y=0,Z=0)
    DriverItem=(Type=Bone,Name="spine_04")
    FalloffRegionScaleFactors=(PositiveWidth=1,NegativeWidth=1,PositiveHeight=1,NegativeHeight=0.7)
    FalloffSize=0.15
    OptionalParentItem=(Type=Bone,Name="spine_01")
    RotationOffset=(X=-70,Y=0,Z=-90)
    ExecutePin=@RigUnit_SequenceAggregate.F
   )
   > &RigUnit_ModifyTransforms(
    ItemToModify=((Item=(Type=Bone,Name="dyn_cape_f_l_base"),Transform=(Rotation=(X=0,Y=0,Z=0,W=1),Translation=(X=0,Y=0,Z=6),Scale3D=(X=1,Y=1,Z=1))))
    Weight=@n5.OutputParam
   )
   > &RigUnit_ModifyTransforms(
    ItemToModify=((Item=(Type=Bone,Name="dyn_cape_s_l_base"),Transform=(Rotation=(X=0,Y=0,Z=0,W=1),Translation=(X=0,Y=0,Z=6),Scale3D=(X=1,Y=1,Z=1))))
    Weight=@n5.OutputParam
   )
   > n6 = &RigUnit_SphericalPoseReader(
    ActiveRegionScaleFactors=(PositiveWidth=1,NegativeWidth=1,PositiveHeight=1,NegativeHeight=1)
    ActiveRegionSize=0.07
    Debug=(bDrawDebug=false,DebugScale=25,DebugSegments=20,DebugThickness=0.25)
    DriverAxis=(X=-1,Y=0,Z=0)
    DriverItem=(Type=Bone,Name="spine_04")
    FalloffRegionScaleFactors=(PositiveWidth=1,NegativeWidth=1,PositiveHeight=1,NegativeHeight=0.7)
    FalloffSize=0.15
    OptionalParentItem=(Type=Bone,Name="spine_01")
    RotationOffset=(X=-115,Y=0,Z=-90)
   )
   > &RigUnit_ModifyTransforms(
    ItemToModify=((Item=(Type=Bone,Name="dyn_cape_f_r_base"),Transform=(Rotation=(X=0,Y=0,Z=0,W=1),Translation=(X=0,Y=0,Z=3),Scale3D=(X=1,Y=1,Z=1))))
    Weight=@n6.OutputParam
   )
   > &RigUnit_ModifyTransforms(
    ItemToModify=((Item=(Type=Bone,Name="dyn_cape_s_r_base"),Transform=(Rotation=(X=0,Y=0,Z=0,W=1),Translation=(X=0,Y=0,Z=3),Scale3D=(X=1,Y=1,Z=1))))
    Weight=@n6.OutputParam
   )
   > &RigUnit_ModifyTransforms(
    ItemToModify=((Item=(Type=Bone,Name="dyn_cape_b_r_base"),Transform=(Rotation=(X=0,Y=0,Z=0,W=1),Translation=(X=0,Y=-3,Z=0),Scale3D=(X=1,Y=1,Z=1))))
    Weight=@n6.OutputParam
   )
   > n7 = &RigUnit_SphericalPoseReader(
    ActiveRegionScaleFactors=(PositiveWidth=1,NegativeWidth=1,PositiveHeight=1,NegativeHeight=1)
    ActiveRegionSize=0.07
    Debug=(bDrawDebug=false,DebugScale=25,DebugSegments=20,DebugThickness=0.25)
    DriverAxis=(X=-1,Y=0,Z=0)
    DriverItem=(Type=Bone,Name="spine_04")
    FalloffRegionScaleFactors=(PositiveWidth=1,NegativeWidth=1,PositiveHeight=1,NegativeHeight=0.7)
    FalloffSize=0.15
    OptionalParentItem=(Type=Bone,Name="spine_01")
    RotationOffset=(X=-110,Y=0,Z=-40)
   )
   > &RigUnit_ModifyTransforms(
    ItemToModify=((Item=(Type=Bone,Name="dyn_cape_s_r_base"),Transform=(Rotation=(X=0,Y=0,Z=0,W=1),Translation=(X=0,Y=4,Z=8),Scale3D=(X=1,Y=1,Z=1))))
    Weight=@n7.OutputParam
   )
   > &RigUnit_ModifyTransforms(
    ItemToModify=((Item=(Type=Bone,Name="dyn_cape_b_r_base"),Transform=(Rotation=(X=0,Y=0,Z=0,W=1),Translation=(X=0,Y=-5,Z=0),Scale3D=(X=1,Y=1,Z=1))))
    Weight=@n7.OutputParam
   )
   &RigUnit_ParentConstraint(
    AdvancedSettings=(RotationOrderForFilter=XZY)
    Child=(Type=Null,Name="Calf_l")
    Parents=((Item=(Type=Bone,Name="calf_l"),Weight=1))
    Weight=1
    bMaintainOffset=false
    ExecutePin=@BranchNode_RigVMUnitNode.True
   )
   > &RigUnit_ParentConstraint(
    AdvancedSettings=(RotationOrderForFilter=XZY)
    Child=(Type=Null,Name="Calf_r")
    Parents=((Item=(Type=Bone,Name="calf_r"),Weight=1))
    Weight=1
    bMaintainOffset=false
   )
   n8 = &RigUnit_GetTransform(Item=(Type=Null,Name="Calf_l"))
   n9 = &RigUnit_GetTransform(Item=(Type=Null,Name="Calf_r"))
   > &RigUnit_ModifyTransforms(
    ItemToModify=((Item=(Type=Bone,Name="altskin_cape_back"),Transform=(Rotation=(X=0,Y=0,Z=0,W=1),Translation=(X=0,Y=0,Z=0),Scale3D=(X=1,Y=1,Z=1))))
    Weight=1
    ItemToModify.0.Transform.Translation.Y=&RigVMFunction_MathDoubleMul(B=-0.15, A=&RigVMFunction_MathVectorDistance(A=@n8.Transform.Translation, B=@n9.Transform.Translation).Result).Result
   )
  ]}
]
hierarchy [
 root {loc=<0 0 0>, rot=<0 0 0 1>}
 pelvis {parent=root, loc=<0 -0.7071 93.1191>, rot=<0.0127 0.707 -0.0127 -0.707>}
 spine_01 {parent=pelvis, loc=<6.7494 -0.6388 0>, rot=<0 0 0.034 -0.9994>}
 spine_02 {parent=spine_01, loc=<7.2518 -0.035 0>, rot=<0 0 0.0238 -0.9997>}
 spine_03 {parent=spine_02, loc=<6.0104 0 0>, rot=<0 0 0.0149 -0.9999>}
 spine_04 {parent=spine_03, loc=<8.1024 0 0>, rot=<0 0 -0.0132 -0.9999>}
 spine_05 {parent=spine_04, loc=<10.0146 0 0>, rot=<0 0 -0.0015 -1>}
 attach_cape {parent=spine_05, loc=<9.017 -11.714 0>, rot=<0.0616 -0.7044 -0.0616 -0.7044>}
 neck_01 {parent=spine_05, loc=<11.3138 0.1529 0>, rot=<0 0 -0.2003 -0.9797>}
 neck_02 {parent=neck_01, loc=<4.5627 0 0>, rot=<0 0 0.0504 -0.9987>}
 head {parent=neck_02, loc=<5.3914 0 0>, rot=<0 0 0.115 -0.9934>}
 attach_backpack {parent=spine_05, loc=<-0.123 -13.473 0>, rot=<-0.0093 -0.707 0.0093 -0.707>}
 clavicle_l {parent=spine_05, loc=<4.3575 -0.1466 -2.3223>, rot=<0.0465 -0.7289 -0.009 -0.6829>}
 upperarm_l {parent=clavicle_l, loc=<15.5111 0 0>, rot=<0.0226 -0.3234 -0.0088 -0.9459>}
 lowerarm_l {parent=upperarm_l, loc=<26.6514 0 0>, rot=<0.0391 -0.0099 -0.2453 -0.9686>}
 hand_l {parent=lowerarm_l, loc=<22.3491 0 0>, rot=<-0.7081 -0.0507 0.0975 -0.6975>}
 thumb_01_l {parent=hand_l, loc=<1.3118 -0.3191 -2.459>, rot=<0.6546 -0.4042 0.0664 -0.6354>}
 thumb_02_l {parent=thumb_01_l, loc=<5.9336 -0.0448 0.0401>, rot=<-0.0037 0.0213 0.298 -0.9543>}
 thumb_03_l {parent=thumb_02_l, loc=<2.2223 0.3682 -0.0907>, rot=<0.0118 -0.013 0.0066 -0.9998>}
 index_metacarpal_l {parent=hand_l, loc=<1.6535 0.0183 -2.96>, rot=<0.0158 -0.0124 -0.0173 -0.9996>}
 index_01_l {parent=index_metacarpal_l, loc=<8.3008 0.0357 -0.0524>, rot=<-0.0044 0.0485 0.0924 -0.9945>}
 index_02_l {parent=index_01_l, loc=<3.3587 -0.3004 0.0452>, rot=<-0.002 0.0027 0.3225 -0.9466>}
 index_03_l {parent=index_02_l, loc=<2.5578 -0.123 -0.0252>, rot=<-0.002 0.0067 0.2558 -0.9667>}
 middle_metacarpal_l {parent=hand_l, loc=<2.1856 -0.1725 -0.5984>, rot=<0.006 0.0168 -0.0329 -0.9993>}
 middle_01_l {parent=middle_metacarpal_l, loc=<7.3275 -0.2447 0.0129>, rot=<-0.0208 0.0042 0.0858 -0.9961>}
 middle_02_l {parent=middle_01_l, loc=<4.0123 0.1401 0.1175>, rot=<-0.0218 0.0258 0.3512 -0.9357>}
 middle_03_l {parent=middle_02_l, loc=<2.6077 0.0534 0.1373>, rot=<0.0082 0.0188 0.1708 -0.9851>}
 ring_metacarpal_l {parent=hand_l, loc=<2.2074 -0.3625 1.3638>, rot=<-0.0284 0.0722 -0.0458 -0.9959>}
 ring_01_l {parent=ring_metacarpal_l, loc=<7.0295 -0.6197 -0.1446>, rot=<-0.0489 -0.0252 0.1892 -0.9804>}
 ring_02_l {parent=ring_01_l, loc=<3.6174 0.4647 -0.1027>, rot=<0 0 0.2806 -0.9598>}
 ring_03_l {parent=ring_02_l, loc=<2.4152 0.114 0.0199>, rot=<0 0 0.1189 -0.9929>}
 pinky_metacarpal_l {parent=hand_l, loc=<1.7616 -0.9753 3.3814>, rot=<-0.0397 0.0814 -0.069 -0.9935>}
 pinky_01_l {parent=pinky_metacarpal_l, loc=<7.1709 -0.4831 -0.1094>, rot=<-0.0654 -0.0035 0.2242 -0.9723>}
 pinky_02_l {parent=pinky_01_l, loc=<2.8274 0.0893 -0.1043>, rot=<0.0037 0.0011 0.2854 -0.9584>}
 pinky_03_l {parent=pinky_02_l, loc=<2.2046 0.0307 -0.0078>, rot=<-0.0002 0.0098 0.1445 -0.9895>}
 lowerarm_twist_02_l {parent=lowerarm_l, loc=<9.62 0 0>, rot=<0 0 0 1>}
 lowerarm_twist_01_l {parent=lowerarm_l, loc=<17.855 0 0>, rot=<0 0 0 1>}
 upperarm_twist_01_l {parent=upperarm_l, loc=<0.0041 0 0>, rot=<0 0 0 -1>}
 deform_upperarm_bk_base_l {parent=upperarm_twist_01_l, loc=<1.0813 -7.2193 -1.9453>, rot=<0 0 0 1>}
 deform_upperarm_bk_l {parent=deform_upperarm_bk_base_l, loc=<0 0 0>, rot=<0 0 0 1>}
 deform_upperarm_fr_base_l {parent=upperarm_twist_01_l, loc=<-0.3691 7.0506 -2.765>, rot=<0 0 0 -1>}
 deform_upperarm_fr_l {parent=deform_upperarm_fr_base_l, loc=<0 0 0>, rot=<0 0 0 1>}
 deform_upperarm_up_base_l {parent=upperarm_twist_01_l, loc=<-3.2216 -1.1391 3.3311>, rot=<0 0 0 -1>}
 deform_upperarm_up_l {parent=deform_upperarm_up_base_l, loc=<0 0 0>, rot=<0 0 0 1>}
 upperarm_twist_02_l {parent=upperarm_l, loc=<13.7941 0 0>, rot=<0 0 0 1>}
 clavicle_r {parent=spine_05, loc=<4.3575 -0.1466 2.3223>, rot=<0.7289 0.0465 -0.6829 0.009>}
 upperarm_r {parent=clavicle_r, loc=<-15.5111 0 0>, rot=<0.0226 -0.3234 -0.0088 -0.9459>}
 lowerarm_r {parent=upperarm_r, loc=<-26.6514 0 0>, rot=<0.0391 -0.0099 -0.2453 -0.9686>}
 hand_r {parent=lowerarm_r, loc=<-22.3491 0 0>, rot=<-0.7081 -0.0507 0.0975 -0.6975>}
 thumb_01_r {parent=hand_r, loc=<-1.3118 0.3191 2.459>, rot=<0.6546 -0.4042 0.0664 -0.6354>}
 thumb_02_r {parent=thumb_01_r, loc=<-5.9336 0.0448 -0.0401>, rot=<-0.0037 0.0213 0.298 -0.9543>}
 thumb_03_r {parent=thumb_02_r, loc=<-2.2223 -0.3682 0.0907>, rot=<0.0118 -0.013 0.0066 -0.9998>}
 index_metacarpal_r {parent=hand_r, loc=<-1.6535 -0.0183 2.96>, rot=<0.0158 -0.0124 -0.0173 -0.9996>}
 index_01_r {parent=index_metacarpal_r, loc=<-8.3008 -0.0357 0.0524>, rot=<-0.0044 0.0485 0.0924 -0.9945>}
 index_02_r {parent=index_01_r, loc=<-3.3587 0.3004 -0.0452>, rot=<-0.002 0.0027 0.3225 -0.9466>}
 index_03_r {parent=index_02_r, loc=<-2.5578 0.123 0.0252>, rot=<-0.002 0.0067 0.2558 -0.9667>}
 middle_metacarpal_r {parent=hand_r, loc=<-2.1856 0.1725 0.5984>, rot=<0.006 0.0168 -0.0329 -0.9993>}
 middle_01_r {parent=middle_metacarpal_r, loc=<-7.3275 0.2447 -0.0129>, rot=<-0.0208 0.0042 0.0858 -0.9961>}
 middle_02_r {parent=middle_01_r, loc=<-4.0123 -0.1401 -0.1175>, rot=<-0.0218 0.0258 0.3512 -0.9357>}
 middle_03_r {parent=middle_02_r, loc=<-2.6077 -0.0534 -0.1373>, rot=<0.0082 0.0188 0.1708 -0.9851>}
 ring_metacarpal_r {parent=hand_r, loc=<-2.2074 0.3625 -1.3638>, rot=<-0.0284 0.0722 -0.0458 -0.9959>}
 ring_01_r {parent=ring_metacarpal_r, loc=<-7.0295 0.6197 0.1446>, rot=<-0.0489 -0.0252 0.1892 -0.9804>}
 ring_02_r {parent=ring_01_r, loc=<-3.6174 -0.4647 0.1027>, rot=<0 0 0.2806 -0.9598>}
 ring_03_r {parent=ring_02_r, loc=<-2.4152 -0.114 -0.0199>, rot=<0 0 0.1189 -0.9929>}
 pinky_metacarpal_r {parent=hand_r, loc=<-1.7616 0.9753 -3.3814>, rot=<-0.0397 0.0814 -0.069 -0.9935>}
 pinky_01_r {parent=pinky_metacarpal_r, loc=<-7.1709 0.4831 0.1094>, rot=<-0.0654 -0.0035 0.2242 -0.9723>}
 pinky_02_r {parent=pinky_01_r, loc=<-2.8274 -0.0893 0.1043>, rot=<0.0037 0.0011 0.2854 -0.9584>}
 pinky_03_r {parent=pinky_02_r, loc=<-2.2046 -0.0307 0.0078>, rot=<-0.0002 0.0098 0.1445 -0.9895>}
 lowerarm_twist_01_r {parent=lowerarm_r, loc=<-17.855 0 0>, rot=<0 0 0 1>}
 lowerarm_twist_02_r {parent=lowerarm_r, loc=<-9.62 0 0>, rot=<0 0 0 1>}
 upperarm_twist_01_r {parent=upperarm_r, loc=<-0.0041 0 0>, rot=<0 0 0 -1>}
 deform_upperarm_bk_base_r {parent=upperarm_twist_01_r, loc=<-1.0814 7.2192 1.9456>, rot=<0 0 0 1>}
 deform_upperarm_bk_r {parent=deform_upperarm_bk_base_r, loc=<0 0 0>, rot=<0 0 0 1>}
 deform_upperarm_fr_base_r {parent=upperarm_twist_01_r, loc=<0.3689 -7.0507 2.7652>, rot=<0 0 0 -1>}
 deform_upperarm_fr_r {parent=deform_upperarm_fr_base_r, loc=<0 0 0>, rot=<0 0 0 1>}
 deform_upperarm_up_base_r {parent=upperarm_twist_01_r, loc=<3.2214 1.1391 -3.3308>, rot=<0 0 0 -1>}
 deform_upperarm_up_r {parent=deform_upperarm_up_base_r, loc=<0 0 0>, rot=<0 0 0 -1>}
 upperarm_twist_02_r {parent=upperarm_r, loc=<-13.7941 0 0>, rot=<0 0 0 1>}
 dyn_hood_base {parent=spine_05, loc=<12.7125 -10.9781 -0.0219>, rot=<0.2588 -0.9659 0 0>}
 dyn_hood_01 {parent=dyn_hood_base, loc=<0 0 0>, rot=<0 0 0 -1>}
 dyn_hood_02 {parent=dyn_hood_01, loc=<5.6528 -0.3392 -0.0381>, rot=<0 0 0.0872 -0.9962>}
 Hood_Corrective {parent=dyn_hood_base, loc=<-5.7385 -4.3677 -0.0219>, rot=<0 0 0 1>}
 cape_spine_corrective {parent=spine_04, loc=<6 -13.4678 0>, rot=<0 -0.7071 0 -0.7071>}
 dyn_cape_corrective_f_l_base {parent=spine_03, loc=<2.1837 9.026 -13.06>, rot=<-0.1933 -0.6802 -0.2665 -0.655>}
 dyn_cape_corrective_f_l {parent=dyn_cape_corrective_f_l_base, loc=<0 0 0>, rot=<0 0 0 1>}
 dyn_cape_corrective_f_r_base {parent=spine_03, loc=<2.6561 9.0257 13.06>, rot=<0.2665 -0.655 0.1933 -0.6802>}
 dyn_cape_corrective_f_r {parent=dyn_cape_corrective_f_r_base, loc=<0 0 0>, rot=<0 0 0 1>}
 thigh_l {parent=pelvis, loc=<-2.5693 -0.0109 -7.7501>, rot=<0.0012 0.0572 0.0155 -0.9982>}
 calf_l {parent=thigh_l, loc=<-39.0485 0 0>, rot=<0.0006 -0.0161 -0.0391 -0.9991>}
 foot_l {parent=calf_l, loc=<-42.561 0.1077 0.4292>, rot=<0.0418 -0.0455 0.0343 -0.9975>}
 ball_l {parent=foot_l, loc=<-6.1598 13.8264 0.6806>, rot=<0 0 -0.7071 -0.7071>}
 dyn_Heelloop_l_03 {parent=foot_l, loc=<3.5391 -5.1886 0.0009>, rot=<-0.6999 0.2185 0.2352 -0.638>}
 dyn_Heelloop_l_02 {parent=dyn_Heelloop_l_03, loc=<1.4495 0 0.3833>, rot=<0 0 0 -1>}
 dyn_Heelloop_l_01 {parent=dyn_Heelloop_l_02, loc=<2.1782 0 0.2058>, rot=<0 0 0 1>}
 calf_twist_02_l {parent=calf_l, loc=<-18.8064 -0.0967 0.0397>, rot=<0.0237 0.0095 -0.0035 -0.9997>}
 calf_twist_01_l {parent=calf_l, loc=<-35.813 -0.0967 0.0397>, rot=<0.0237 0.0095 -0.0035 -0.9997>}
 deform_knee_in_base_l {parent=calf_l, loc=<1.298 -5.2957 -0.0496>, rot=<0 0 0 -1>}
 deform_knee_in_l {parent=deform_knee_in_base_l, loc=<0 0 0>, rot=<0 0 0 -1>}
 deform_knee_out_base_l {parent=calf_l, loc=<0 4 0>, rot=<0 0 0 -1>}
 deform_knee_out_l {parent=deform_knee_out_base_l, loc=<0 0 0>, rot=<0 0 0 -1>}
 thigh_twist_01_l {parent=thigh_l, loc=<-21.6303 0 0>, rot=<0 0 0 1>}
 dyn_short_l_fr_base {parent=thigh_twist_01_l, loc=<2.0583 9.9965 1.4005>, rot=<1 0 0 -0.0013>}
 dyn_short_l_fr_01 {parent=dyn_short_l_fr_base, loc=<-7.428 0 0>, rot=<0 0 0 1>}
 dyn_short_l_fr {parent=dyn_short_l_fr_01, loc=<0 0 0>, rot=<0 0 0 1>}
 dyn_short_l_sl_base {parent=thigh_twist_01_l, loc=<2.0583 -1.2217 -10.0031>, rot=<-0.708 0 0 -0.7062>}
 dyn_short_l_sl_01 {parent=dyn_short_l_sl_base, loc=<-7.428 0 0>, rot=<0 0 0 1>}
 dyn_short_l_sl {parent=dyn_short_l_sl_01, loc=<0 0 0>, rot=<0 0 0 1>}
 dyn_short_l_b_base {parent=thigh_twist_01_l, loc=<2.058 -9.9839 -1.25>, rot=<-0.0013 0 0 -1>}
 dyn_short_l_b_01 {parent=dyn_short_l_b_base, loc=<-7.428 0 0>, rot=<0 0 0 1>}
 dyn_short_l_b {parent=dyn_short_l_b_01, loc=<0 0 0>, rot=<0 0 0 1>}
 dyn_short_l_sr_base {parent=thigh_twist_01_l, loc=<2.0583 -1.6269 9.9959>, rot=<0.7062 0 0 -0.708>}
 dyn_short_l_sr_01 {parent=dyn_short_l_sr_base, loc=<-7.428 0 0>, rot=<0 0 0 1>}
 dyn_short_l_sr {parent=dyn_short_l_sr_01, loc=<0 0 0>, rot=<0 0 0 1>}
 deform_groin_base_l {parent=thigh_l, loc=<-2.1908 8.1142 -1.7442>, rot=<0 0 0 1>}
 deform_groin_l {parent=deform_groin_base_l, loc=<0 0 0>, rot=<0 0 0 1>}
 thigh_r {parent=pelvis, loc=<-2.5693 -0.0109 7.7501>, rot=<0.0572 -0.0012 0.9982 0.0155>}
 calf_r {parent=thigh_r, loc=<39.0485 0 0>, rot=<0.0006 -0.0161 -0.0391 -0.9991>}
 foot_r {parent=calf_r, loc=<42.561 -0.1077 -0.4292>, rot=<0.0418 -0.0455 0.0343 -0.9975>}
 ball_r {parent=foot_r, loc=<6.1598 -13.8264 -0.6806>, rot=<0 0 -0.7071 -0.7071>}
 dyn_Heelloop_r_03 {parent=foot_r, loc=<-3.5349 5.1885 -0.0008>, rot=<-0.6999 0.2185 0.2352 -0.638>}
 dyn_Heelloop_r_02 {parent=dyn_Heelloop_r_03, loc=<-1.4503 0 -0.3697>, rot=<0 0 0 -1>}
 dyn_Heelloop_r_01 {parent=dyn_Heelloop_r_02, loc=<-2.1859 0 -0.2134>, rot=<0 0 0 1>}
 calf_twist_02_r {parent=calf_r, loc=<18.8064 0.0967 -0.0397>, rot=<0.0237 0.0095 -0.0035 -0.9997>}
 calf_twist_01_r {parent=calf_r, loc=<35.813 0.0967 -0.0397>, rot=<0.0237 0.0095 -0.0035 -0.9997>}
 deform_knee_in_base_r {parent=calf_r, loc=<-1.298 5.2957 0.0496>, rot=<0 0 0 -1>}
 deform_knee_in_r {parent=deform_knee_in_base_r, loc=<0 0 0>, rot=<0 0 0 -1>}
 deform_knee_out_base_r {parent=calf_r, loc=<0 -4 0>, rot=<0 0 0 -1>}
 deform_knee_out_r {parent=deform_knee_out_base_r, loc=<0 0 0>, rot=<0 0 0 -1>}
 thigh_twist_01_r {parent=thigh_r, loc=<21.6303 0 0>, rot=<0 0 0 1>}
 dyn_short_r_fr_base {parent=thigh_twist_01_r, loc=<-1.9953 -10 -1.3752>, rot=<0 0 0 1>}
 dyn_short_r_fr_01 {parent=dyn_short_r_fr_base, loc=<7.3651 0 0>, rot=<0 0 0 1>}
 dyn_short_r_fr {parent=dyn_short_r_fr_01, loc=<0 0 0>, rot=<0 0 0 1>}
 dyn_short_r_sr_base {parent=thigh_twist_01_r, loc=<-1.9954 1.6016 -10>, rot=<-0.7071 0 0 -0.7071>}
 dyn_short_r_sr_01 {parent=dyn_short_r_sr_base, loc=<7.3651 0 0>, rot=<0 0 0 1>}
 dyn_short_r_sr {parent=dyn_short_r_sr_01, loc=<0 0 0>, rot=<0 0 0 1>}
 dyn_short_r_b_base {parent=thigh_twist_01_r, loc=<-1.9954 10 1.2247>, rot=<1 0 0 0>}
 dyn_short_r_b_01 {parent=dyn_short_r_b_base, loc=<7.3651 0 0>, rot=<0 0 0 1>}
 dyn_short_r_b {parent=dyn_short_r_b_01, loc=<0 0 0>, rot=<0 0 0 1>}
 dyn_short_r_sl_base {parent=thigh_twist_01_r, loc=<-1.9953 1.247 10>, rot=<0.7071 0 0 -0.7071>}
 dyn_short_r_sl_01 {parent=dyn_short_r_sl_base, loc=<7.3651 0 0>, rot=<0 0 0 1>}
 dyn_short_r_sl {parent=dyn_short_r_sl_01, loc=<0 0 0>, rot=<0 0 0 1>}
 deform_groin_base_r {parent=thigh_r, loc=<2.1908 -8.1142 1.7442>, rot=<0 0 0 1>}
 deform_groin_r {parent=deform_groin_base_r, loc=<0 0 0>, rot=<0 0 0 1>}
 deform_glute_base_l {parent=pelvis, loc=<-8.8064 -8.8222 -8.4287>, rot=<0.0275 0.0501 -0.4472 -0.8926>}
 deform_glute_l {parent=deform_glute_base_l, loc=<0 0 0>, rot=<0 0 0 1>}
 deform_glute_base_r {parent=pelvis, loc=<-8.8062 -8.8222 8.4287>, rot=<0.0501 -0.0275 0.8926 -0.4472>}
 deform_glute_r {parent=deform_glute_base_r, loc=<0 0 0>, rot=<0 0 0 1>}
 altskin_cape_back {parent=pelvis, loc=<-26.72 -17.115 0>, rot=<0 0 0 1>}
 dyn_cape_f_r_base {parent=pelvis, loc=<13.3067 5.8012 13.483>, rot=<0.1018 0.2946 0.9501 -0.008>}
 dyn_cape_f_r_01 {parent=dyn_cape_f_r_base, loc=<0 0 0>, rot=<0 0 0 1>}
 dyn_strap_r_01 {parent=dyn_cape_f_r_01, loc=<0 0 1.5>, rot=<0.0306 0.0064 -0.0083 -0.9995>}
 dyn_strap_r_02 {parent=dyn_strap_r_01, loc=<17.799 0 0>, rot=<0 0.0092 0 -1>}
 dyn_strap_r_03 {parent=dyn_strap_r_02, loc=<20 0 0>, rot=<0 0.0068 0 -1>}
 dyn_strap_r_04 {parent=dyn_strap_r_03, loc=<14 0 0>, rot=<0 0 0 1>}
 dyn_strap_r_05 {parent=dyn_strap_r_04, loc=<12 0 0>, rot=<0 0 0 1>}
 dyn_cape_f_r_02 {parent=dyn_cape_f_r_01, loc=<17.799 0 0>, rot=<0 -0.0017 0.0048 -1>}
 dyn_cape_f_r_03 {parent=dyn_cape_f_r_02, loc=<20 0 0>, rot=<0 0.0052 0 -1>}
 dyn_cape_s_r_base {parent=pelvis, loc=<12.4116 -6.1564 15.3014>, rot=<0.0935 -0.1598 0.9827 0.0043>}
 dyn_cape_s_r_01 {parent=dyn_cape_s_r_base, loc=<0.0003 0 0>, rot=<0 0 0 1>}
 dyn_cape_s_r_02 {parent=dyn_cape_s_r_01, loc=<13.5443 0 0>, rot=<0.0007 0.018 0.0021 -0.9998>}
 dyn_cape_s_r_03 {parent=dyn_cape_s_r_02, loc=<23.4406 0.0391 -0.3402>, rot=<0 0 0 1>}
 dyn_cape_b_r_base {parent=pelvis, loc=<12.3905 -12.1677 6>, rot=<0.0653 -0.6228 0.7795 -0.0158>}
 dyn_cape_b_r_01 {parent=dyn_cape_b_r_base, loc=<0.0005 0 0>, rot=<0 0 0 1>}
 dyn_cape_b_r_02 {parent=dyn_cape_b_r_01, loc=<15.2519 -0.0001 0>, rot=<0 0.0092 0 -1>}
 dyn_cape_b_r_03 {parent=dyn_cape_b_r_02, loc=<21.7441 0 -0.1428>, rot=<0 0 0 1>}
 dyn_cape_b_l_base {parent=pelvis, loc=<12.39 -12.1681 -6>, rot=<0.0184 -0.7672 0.6383 -0.0598>}
 dyn_cape_b_l_01 {parent=dyn_cape_b_l_base, loc=<0 0 0>, rot=<0 0 0 1>}
 dyn_cape_b_l_02 {parent=dyn_cape_b_l_01, loc=<16.4234 0 0>, rot=<0 0.0092 0 -1>}
 dyn_cape_b_l_03 {parent=dyn_cape_b_l_02, loc=<20.5731 0 -0.1212>, rot=<0 0 0 1>}
 dyn_cape_s_l_base {parent=pelvis, loc=<12.4444 -5.6449 -15.222>, rot=<-0.0162 -0.984 0.1501 -0.0947>}
 dyn_cape_s_l_01 {parent=dyn_cape_s_l_base, loc=<-0.0002 0 0>, rot=<0 0 0 1>}
 dyn_cape_s_l_02 {parent=dyn_cape_s_l_01, loc=<16.668 0 0>, rot=<0.0007 0.018 0.0021 -0.9998>}
 dyn_cape_s_l_03 {parent=dyn_cape_s_l_02, loc=<20.3195 0.0262 -0.2277>, rot=<0 0 0 1>}
 dyn_cape_f_l_base {parent=pelvis, loc=<13.3668 5.83 -13.483>, rot=<0.0107 -0.9381 -0.3336 -0.0929>}
 dyn_cape_f_l_01 {parent=dyn_cape_f_l_base, loc=<0 0 0>, rot=<0 0 0 1>}
 dyn_strap_l_01 {parent=dyn_cape_f_l_01, loc=<0 0 1.5>, rot=<-0.0002 0.0123 0.0151 -0.9998>}
 dyn_strap_l_02 {parent=dyn_strap_l_01, loc=<17.799 0 0>, rot=<0 0.0101 0 -0.9999>}
 dyn_strap_l_03 {parent=dyn_strap_l_02, loc=<20 0 0>, rot=<0 0.0075 0 -1>}
 dyn_strap_l_04 {parent=dyn_strap_l_03, loc=<14 0 0>, rot=<0 0 0 -1>}
 dyn_strap_l_05 {parent=dyn_strap_l_04, loc=<12 0 0>, rot=<0 0 0 -1>}
 dyn_cape_f_l_02 {parent=dyn_cape_f_l_01, loc=<17.7992 0 0>, rot=<-0.001 0.0052 0.0024 -1>}
 dyn_cape_f_l_03 {parent=dyn_cape_f_l_02, loc=<20 0.025 -0.0543>, rot=<0 0.0052 0 -1>}
 calf_l {parent=pelvis, type=Null, loc=<0 0 0>, rot=<0 0 0 1>}
 calf_r {parent=pelvis, type=Null, loc=<0 0 0>, rot=<0 0 0 1>}
 floor_joint {parent=root, loc=<0 0 0>, rot=<0 0 0 1>}
]
variables [
 Is FrontEnd {type=bool, default=False}
 Is enabled rigid body {type=bool, default=False}
]
