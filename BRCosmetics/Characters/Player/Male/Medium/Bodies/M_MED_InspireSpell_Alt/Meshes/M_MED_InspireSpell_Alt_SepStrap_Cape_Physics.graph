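# Physics asset for the separated-strap cape of the M_MED_InspireSpell_Alt male medium player cosmetic body, defining capsule bodies for its straps and front cape bones.
# Left and right five-bone strap chains use swing-limited, twist-locked constraints, with each root angularly locked to the first front cape bone on its side.
# Three multi-capsule cape bodies per side have no constraints between them.
# A left calf capsule and a huge floor_joint capsule add further collision shapes.
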PhysicsAsset /BRCosmetics/Characters/Player/Male/Medium/Bodies/M_MED_InspireSpell_Alt/Meshes/M_MED_InspireSpell_Alt_SepStrap_Cape_Physics
bodies {
 calf_l [{
   shapeType=Capsule
   length=31.2926
   radius=11.6366
   shapeName=calf_l_capsule
   center=<-13.8666 -1.8957 -0.0706>
   rotation=<rot 0 79.5515 -89.8774>
  }]
 dyn_cape_f_l_01 [
  {
   shapeType=Capsule
   length=17.7376
   radius=3.2624
   shapeName=dyn_cape_f_l_01_capsule
   center=<8.7795 -4.6462 -1.6122>
   rotation=<rot 1.4846 17.2302 85.2249>
  }
  {shapeType=Capsule, length=17.7376, radius=2.2624, shapeName=dyn_cape_f_l_01_capsule, center=<5.5622 4.3396 0>, rotation=<rot 90 0 0>}
 ]
 dyn_cape_f_l_02 [
  {
   shapeType=Capsule
   length=17.7376
   radius=3.2624
   shapeName=dyn_cape_f_l_02_capsule
   center=<0 -4.6629 -1.8813>
   rotation=<rot -3.3833 -42.5117 86.3164>
  }
  {shapeType=Capsule, length=17.7376, radius=2.2624, shapeName=dyn_cape_f_l_02_capsule, center=<5.5622 4.3443 -1.0804>, rotation=<rot 90 0 0>}
  {
   shapeType=Capsule
   length=16.9688
   radius=4.0312
   shapeName=dyn_cape_f_l_02_capsule
   center=<12.8223 -4.5655 -2.2651>
   rotation=<rot 79.7596 -90.0341 -0.0445>
  }
 ]
 dyn_cape_f_l_03 [
  {
   shapeType=Capsule
   length=18.7701
   radius=4.2299
   shapeName=dyn_cape_f_l_03_capsule
   center=<-0.0187 -4.8122 -2.3037>
   rotation=<rot -4.6333 -47.9243 86.6649>
  }
  {shapeType=Capsule, length=20, radius=3, shapeName=dyn_cape_f_l_03_capsule, center=<5.3584 4.3099 -0.8465>, rotation=<rot 90 0 0>}
  {
   shapeType=Capsule
   length=16.1455
   radius=5.5921
   shapeName=dyn_cape_f_l_03_capsule
   center=<12.4716 -5.1493 -1.693>
   rotation=<rot 0.0001 0.0001 79.6014>
  }
  {
   shapeType=Capsule
   length=40.9826
   radius=4
   shapeName=dyn_cape_f_l_03_capsule
   center=<-8.3626 -13.1002 -4.2223>
   rotation=<rot 90 -179.7762 -177.5318>
  }
 ]
 dyn_cape_f_r_01 [
  {shapeType=Capsule, length=14, radius=4, shapeName=dyn_cape_f_r_01_capsule, center=<0 6.9554 0>, rotation=<rot 0 -22.7507 90>}
  {shapeType=Capsule, length=19, radius=3, shapeName=dyn_cape_f_r_01_capsule, center=<4.8057 -2.299 0>, rotation=<rot 90 0 0>}
 ]
 dyn_cape_f_r_02 [
  {shapeType=Capsule, length=21.2792, radius=4.5198, shapeName=dyn_cape_f_r_02_capsule, center=<-2.7387 8.5024 0>, rotation=<rot 0 39.5095 90>}
  {shapeType=Capsule, length=19, radius=4, shapeName=dyn_cape_f_r_02_capsule, center=<4.8057 -2.299 -1.5631>, rotation=<rot 90 0 0>}
  {
   shapeType=Capsule
   length=19
   radius=4
   shapeName=dyn_cape_f_r_02_capsule
   center=<7.8396 10.0955 -1.4198>
   rotation=<rot 81.7666 90.0001 -179.9999>
  }
 ]
 dyn_cape_f_r_03 [
  {shapeType=Capsule, length=20, radius=4, shapeName=dyn_cape_f_r_03_capsule, center=<0.1299 8.6588 -3.2665>, rotation=<rot 0 47.7662 97.2099>}
  {shapeType=Capsule, length=21, radius=3, shapeName=dyn_cape_f_r_03_capsule, center=<4.8057 -2.299 -1.8389>, rotation=<rot 90 0 0>}
  {
   shapeType=Capsule
   length=18.9721
   radius=5.0279
   shapeName=dyn_cape_f_r_03_capsule
   center=<12.4778 8.7606 -3.4734>
   rotation=<rot 80.5032 90.0572 -179.9992>
  }
  {
   shapeType=Capsule
   length=40.9826
   radius=4
   shapeName=dyn_cape_f_r_03_capsule
   center=<-8.2479 18.9313 -2.5909>
   rotation=<rot 86.8002 0.0001 0.0002>
  }
 ]
 dyn_strap_l_01 [{shapeType=Capsule, length=18, radius=3, shapeName=dyn_strap_l_01_capsule, center=<0 0 4.5>, rotation=<rot 90 0 0>}]
 dyn_strap_l_02 [{shapeType=Capsule, length=18, radius=3, shapeName=dyn_strap_l_02_capsule, center=<0 0 4.5>, rotation=<rot 80.0858 -179.9997 -179.9998>}]
 dyn_strap_l_03 [{shapeType=Capsule, length=18, radius=3, shapeName=dyn_strap_l_03_capsule, center=<0 0 4.5>, rotation=<rot 90 0 0>}]
 dyn_strap_l_04 [{shapeType=Capsule, length=18, radius=3, shapeName=dyn_strap_l_04_capsule, center=<0 0 4.5>, rotation=<rot 90 0 0>}]
 dyn_strap_l_05 [{shapeType=Capsule, length=18, radius=3, shapeName=dyn_strap_l_05_capsule, center=<0 0 4.5>, rotation=<rot 90 0 0>}]
 dyn_strap_r_01 [{shapeType=Capsule, length=18, radius=3, shapeName=dyn_strap_r_01_capsule, center=<0 0 4.5>, rotation=<rot 90 0 0>}]
 dyn_strap_r_02 [{shapeType=Capsule, length=18, radius=3, shapeName=dyn_strap_r_02_capsule, center=<0 0 5>, rotation=<rot 83.0608 179.999 179.9993>}]
 dyn_strap_r_03 [{shapeType=Capsule, length=18, radius=3, shapeName=dyn_strap_r_03_capsule, center=<0 0 3>, rotation=<rot 87.3646 0.0003 0.0003>}]
 dyn_strap_r_04 [{shapeType=Capsule, length=18, radius=3, shapeName=dyn_strap_r_04_capsule, center=<0 0 3>, rotation=<rot 90 0 0>}]
 dyn_strap_r_05 [{shapeType=Capsule, length=18, radius=3, shapeName=dyn_strap_r_05_capsule, center=<0 0 3>, rotation=<rot 90 0 0>}]
 floor_joint [{shapeType=Capsule, length=48.3294, radius=165.0692, shapeName=root_capsule, center=<2.6954 -23.3737 -154.0549>, rotation=<rot 0 0 -90>}]
}
constraints [
 {bone1Name=dyn_strap_l_02, bone2Name=dyn_strap_l_01, swing1LimitDegrees=15, swing2LimitDegrees=35, twistLimitDegrees=45, twistMotion=Locked}
 {bone1Name=dyn_strap_l_03, bone2Name=dyn_strap_l_02, swing1LimitDegrees=15, swing2LimitDegrees=30, twistLimitDegrees=45, twistMotion=Locked}
 {bone1Name=dyn_strap_l_04, bone2Name=dyn_strap_l_03, swing1LimitDegrees=30, swing2LimitDegrees=35, twistLimitDegrees=45, twistMotion=Locked}
 {bone1Name=dyn_strap_l_05, bone2Name=dyn_strap_l_04, swing1LimitDegrees=5, swing2LimitDegrees=35, twistLimitDegrees=45, twistMotion=Locked}
 {bone1Name=dyn_strap_r_02, bone2Name=dyn_strap_r_01, swing1LimitDegrees=20, swing2LimitDegrees=35, twistLimitDegrees=45, twistMotion=Locked}
 {bone1Name=dyn_strap_r_03, bone2Name=dyn_strap_r_02, swing1LimitDegrees=20, swing2LimitDegrees=35, twistLimitDegrees=45, twistMotion=Locked}
 {bone1Name=dyn_strap_r_04, bone2Name=dyn_strap_r_03, swing1LimitDegrees=30, swing2LimitDegrees=35, twistLimitDegrees=45, twistMotion=Locked}
 {bone1Name=dyn_strap_r_05, bone2Name=dyn_strap_r_04, swing1LimitDegrees=5, swing2LimitDegrees=35, twistLimitDegrees=45, twistMotion=Locked}
 {
  bone1Name=dyn_strap_l_01
  bone2Name=dyn_cape_f_l_01
  swing1LimitDegrees=45
  swing1Motion=Locked
  swing2LimitDegrees=45
  swing2Motion=Locked
  twistLimitDegrees=45
  twistMotion=Locked
 }
 {
  bone1Name=dyn_strap_r_01
  bone2Name=dyn_cape_f_r_01
  swing1LimitDegrees=45
  swing1Motion=Locked
  swing2LimitDegrees=45
  swing2Motion=Locked
  twistLimitDegrees=45
  twistMotion=Locked
 }
]
