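# Physics asset for the shorts of the M_MED_InspireSpell_Alt male body cosmetic, covering dynamic cloth bones on the left leg only.
# Four two-bone chains (b, fr, sl, sr) hang off a thigh_l capsule.
# Each zero-size base body is angularly locked to the thigh as an anchor.
# Each 01 capsule has tight angular limits against its base, allowing only slight, mostly swing, motion of the fabric.
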PhysicsAsset /BRCosmetics/Characters/Player/Male/Medium/Bodies/M_MED_InspireSpell_Alt/Meshes/M_MED_InspireSpell_Alt_Shorts_Physics
bodies {
 dyn_short_l_b_01 [{shapeType=Capsule, length=7, radius=4, shapeName=dyn_short_l_b_01_capsule, center=<-6.0445 0 0>}]
 dyn_short_l_b_base [{shapeType=Capsule, length=0, radius=0, shapeName=dyn_short_l_b_base_capsule}]
 dyn_short_l_fr_01 [{shapeType=Capsule, length=15, radius=2, shapeName=dyn_short_l_fr_01_capsule, center=<-6 1 0>}]
 dyn_short_l_fr_base [{shapeType=Capsule, length=0, radius=0, shapeName=dyn_short_l_fr_base_capsule}]
 dyn_short_l_sl_01 [{shapeType=Capsule, length=15, radius=2, shapeName=dyn_short_l_sl_01_capsule, center=<-6 1 0>}]
 dyn_short_l_sl_base [{shapeType=Capsule, length=0, radius=0, shapeName=dyn_short_l_sl_base_capsule}]
 dyn_short_l_sr_01 [{shapeType=Capsule, length=15, radius=2, shapeName=dyn_short_l_sr_01_capsule, center=<-6 0 0>}]
 dyn_short_l_sr_base [{shapeType=Capsule, length=0, radius=0, shapeName=dyn_short_l_sr_base_capsule}]
 thigh_l [{
   shapeType=Capsule
   length=23.3262
   radius=6.8019
   shapeName=thigh_l_capsule
   center=<-24.5429 0.0024 -0.1755>
   rotation=<rot 0 95.1034 -89.6962>
  }]
}
constraints [
 {bone1Name=dyn_short_l_sr_01, bone2Name=dyn_short_l_sr_base, swing1LimitDegrees=10, swing2LimitDegrees=0, twistLimitDegrees=0}
 {
  bone1Name=dyn_short_l_sr_base
  bone2Name=thigh_l
  swing1LimitDegrees=45
  swing1Motion=Locked
  swing2LimitDegrees=45
  swing2Motion=Locked
  twistLimitDegrees=45
  twistMotion=Locked
 }
 {bone1Name=dyn_short_l_sl_01, bone2Name=dyn_short_l_sl_base, swing1LimitDegrees=10, swing2LimitDegrees=2, twistLimitDegrees=0}
 {
  bone1Name=dyn_short_l_sl_base
  bone2Name=thigh_l
  swing1LimitDegrees=45
  swing1Motion=Locked
  swing2LimitDegrees=45
  swing2Motion=Locked
  twistLimitDegrees=45
  twistMotion=Locked
 }
 {bone1Name=dyn_short_l_fr_01, bone2Name=dyn_short_l_fr_base, swing1LimitDegrees=10, swing2LimitDegrees=2, twistLimitDegrees=0}
 {
  bone1Name=dyn_short_l_fr_base
  bone2Name=thigh_l
  swing1LimitDegrees=45
  swing1Motion=Locked
  swing2LimitDegrees=45
  swing2Motion=Locked
  twistLimitDegrees=45
  twistMotion=Locked
 }
 {bone1Name=dyn_short_l_b_01, bone2Name=dyn_short_l_b_base, swing1LimitDegrees=5.76, swing2LimitDegrees=5, twistLimitDegrees=5}
 {
  bone1Name=dyn_short_l_b_base
  bone2Name=thigh_l
  swing1LimitDegrees=45
  swing1Motion=Locked
  swing2LimitDegrees=45
  swing2Motion=Locked
  twistLimitDegrees=45
  twistMotion=Locked
 }
]
