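# Physics asset for the straps and front cape of the male medium InspireSpell Alt cosmetic body.
# Identical capsules cover mirrored left and right chains of three cape bones and five strap bones.
# The first three strap bones are constrained to matching cape bones, locked at the first and allowing narrow single-axis swing at the next two.
# The last two continue down the strap with two-axis swing limits, and twist is locked throughout.
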PhysicsAsset /BRCosmetics/Characters/Player/Male/Medium/Bodies/M_MED_InspireSpell_Alt/Meshes/M_MED_InspireSpell_Alt_Strap_Physics
bodies {
 dyn_cape_f_l_01 [{shapeType=Capsule, length=7, radius=3, shapeName=dyn_cape_f_l_01_capsule, rotation=<rot 90 0 0>}]
 dyn_cape_f_l_02 [{shapeType=Capsule, length=7, radius=3, shapeName=dyn_cape_f_l_02_capsule, rotation=<rot 90 0 0>}]
 dyn_cape_f_l_03 [{shapeType=Capsule, length=7, radius=3, shapeName=dyn_cape_f_l_03_capsule, rotation=<rot 90 0 0>}]
 dyn_cape_f_r_01 [{shapeType=Capsule, length=7, radius=3, shapeName=dyn_cape_f_r_01_capsule, rotation=<rot 90 0 0>}]
 dyn_cape_f_r_02 [{shapeType=Capsule, length=7, radius=3, shapeName=dyn_cape_f_r_02_capsule, rotation=<rot 90 0 0>}]
 dyn_cape_f_r_03 [{shapeType=Capsule, length=7, radius=3, shapeName=dyn_cape_f_r_03_capsule, rotation=<rot 90 0 0>}]
 dyn_strap_l_01 [{shapeType=Capsule, length=7, radius=3, shapeName=dyn_strap_l_01_capsule, rotation=<rot 90 0 0>}]
 dyn_strap_l_02 [{shapeType=Capsule, length=7, radius=3, shapeName=dyn_strap_l_02_capsule, rotation=<rot 90 0 0>}]
 dyn_strap_l_03 [{shapeType=Capsule, length=7, radius=3, shapeName=dyn_strap_l_03_capsule, rotation=<rot 90 0 0>}]
 dyn_strap_l_04 [{shapeType=Capsule, length=7, radius=3, shapeName=dyn_strap_l_04_capsule, rotation=<rot 90 0 0>}]
 dyn_strap_l_05 [{shapeType=Capsule, length=7, radius=3, shapeName=dyn_strap_l_05_capsule, rotation=<rot 90 0 0>}]
 dyn_strap_r_01 [{shapeType=Capsule, length=7, radius=3, shapeName=dyn_strap_r_01_capsule, rotation=<rot 90 0 0>}]
 dyn_strap_r_02 [{shapeType=Capsule, length=7, radius=3, shapeName=dyn_strap_r_02_capsule, rotation=<rot 90 0 0>}]
 dyn_strap_r_03 [{shapeType=Capsule, length=7, radius=3, shapeName=dyn_strap_r_03_capsule, rotation=<rot 90 0 0>}]
 dyn_strap_r_04 [{shapeType=Capsule, length=7, radius=3, shapeName=dyn_strap_r_04_capsule, rotation=<rot 90 0 0>}]
 dyn_strap_r_05 [{shapeType=Capsule, length=7, radius=3, shapeName=dyn_strap_r_05_capsule, rotation=<rot 90 0 0>}]
}
constraints [
 {
  bone1Name=dyn_strap_l_01
  bone2Name=dyn_cape_f_l_01
  swing1LimitDegrees=45
  swing1Motion=Locked
  swing2LimitDegrees=45
  swing2Motion=Locked
  twistLimitDegrees=45
  twistMotion=Locked
 }
 {
  bone1Name=dyn_strap_l_02
  bone2Name=dyn_cape_f_l_02
  swing1LimitDegrees=5
  swing2LimitDegrees=45
  swing2Motion=Locked
  twistLimitDegrees=45
  twistMotion=Locked
 }
 {
  bone1Name=dyn_strap_l_03
  bone2Name=dyn_cape_f_l_03
  swing1LimitDegrees=9
  swing2LimitDegrees=45
  swing2Motion=Locked
  twistLimitDegrees=45
  twistMotion=Locked
 }
 {
  bone1Name=dyn_strap_r_01
  bone2Name=dyn_cape_f_r_01
  swing1LimitDegrees=45
  swing1Motion=Locked
  swing2LimitDegrees=45
  swing2Motion=Locked
  twistLimitDegrees=45
  twistMotion=Locked
 }
 {
  bone1Name=dyn_strap_r_02
  bone2Name=dyn_cape_f_r_02
  swing1LimitDegrees=5
  swing2LimitDegrees=45
  swing2Motion=Locked
  twistLimitDegrees=45
  twistMotion=Locked
 }
 {
  bone1Name=dyn_strap_r_03
  bone2Name=dyn_cape_f_r_03
  swing1LimitDegrees=9
  swing2LimitDegrees=45
  swing2Motion=Locked
  twistLimitDegrees=45
  twistMotion=Locked
 }
 {bone1Name=dyn_strap_r_04, bone2Name=dyn_strap_r_03, swing1LimitDegrees=30, swing2LimitDegrees=25, twistLimitDegrees=45, twistMotion=Locked}
 {bone1Name=dyn_strap_r_05, bone2Name=dyn_strap_r_04, swing1LimitDegrees=5, swing2LimitDegrees=30, twistLimitDegrees=45, twistMotion=Locked}
 {bone1Name=dyn_strap_l_05, bone2Name=dyn_strap_l_04, swing1LimitDegrees=5, swing2LimitDegrees=30, twistLimitDegrees=45, twistMotion=Locked}
 {bone1Name=dyn_strap_l_04, bone2Name=dyn_strap_l_03, swing1LimitDegrees=30, swing2LimitDegrees=25, twistLimitDegrees=45, twistMotion=Locked}
]
