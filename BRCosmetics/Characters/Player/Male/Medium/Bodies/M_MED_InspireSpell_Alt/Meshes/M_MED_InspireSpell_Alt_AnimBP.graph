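# Post-process style anim Blueprint for the InspireSpell Alt character body, adding secondary motion to its cape, strap, shorts and hood.
# It copies the parent pose, offsets cape bones for skydiving and knee sliding, then runs a movement Control Rig.
# Chained RigidBody nodes follow, with front-end/gameplay cape switching and variant-gated cape, strap and hood simulation, ending in a second Control Rig.
# The EventGraph updates gravity, simulation settings and material WPO wind/thigh parameters.
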AnimBlueprint /BRCosmetics/Characters/Player/Male/Medium/Bodies/M_MED_InspireSpell_Alt/Meshes/M_MED_InspireSpell_Alt_AnimBP
target_skeleton=&/M_MED_InspireSpell_Alt_Skeleton
graphs [
 AnimGraph {nodes [
   RigidBody!(ComponentPose, bAlphaBoolEnabled:bool=false, OverrideWorldGravity:Vector="0,0,-980", ExternalForce) -> (Pose) {
    alphaInputType=Bool
    bClampLinearTranslationLimitToRefPose=true
    bOverrideWorldGravity=true
    overridePhysicsAsset=&/M_MED_InspireSpell_Alt_Strap_Physics
    simulationSpace=BaseBoneSpace
    alphaBoolBlend {blendInTime=0.5, blendOption=Linear, blendOutTime=0.5}
    baseBoneRef {boneName=pelvis}
    componentAppliedLinearAccClamp=<1800 1800 1800>
    componentLinearAccScale=<0.01 0.001 0.3>
    componentLinearVelScale=<0.08 0.05 0.1>
    overrideWorldGravity {z=-980}
    simSpaceSettings {
     maxAngularAcceleration=150
     maxAngularVelocity=2
     maxLinearAcceleration=250
     maxLinearVelocity=50
     worldAlpha=1
     externalLinearDragV=<0.1 0.1 0.1>
    }
   }
   n1 = Get Enable_Cape_RBAN() -> (Enable_Cape_RBAN:bool)
   n2 = Get Sim Space Settings() -> (Sim Space Settings:Sim Space Settings Structure)
   n3 = Get Gravity Override()
   n4 = RigidBody!(
    ComponentPose=Local To Component(LocalPose=Use cached pose 'Init Transforms'())
    bAlphaBoolEnabled=@n1
    OverrideWorldGravity=@n3
    ExternalForce
    SimSpaceSettings:Sim Space Settings Structure=@n2
   ) -> (Pose) {
    alphaInputType=Bool
    bClampLinearTranslationLimitToRefPose=true
    bOverrideWorldGravity=true
    overridePhysicsAsset=&/M_MED_InspireSpell_Alt_FrontEnd_Cape_Physics
    simulationSpace=BaseBoneSpace
    alphaBoolBlend {blendOption=Linear}
    alphaScaleBiasClamp {bInterpResult=true}
    baseBoneRef {boneName=pelvis}
    componentAppliedLinearAccClamp=<1800 1800 1800>
    componentLinearAccScale=<0.4 0.15 0.15>
    componentLinearVelScale=<0.18 0.15 0.15>
    overrideWorldGravity {z=-980}
    simSpaceSettings {
     maxAngularAcceleration=250
     maxAngularVelocity=2
     maxLinearAcceleration=50
     maxLinearVelocity=120
     velocityScaleZ=0
     worldAlpha=1
     externalLinearDragV=<0.1 0.1>
    }
   }
   n5 = Component To Local(ComponentPose=@n4)
   n6 = RigidBody(
    ComponentPose=Local To Component(LocalPose=Use cached pose 'Init Transforms'())
    bAlphaBoolEnabled=@n1
    OverrideWorldGravity=@n3
    SimSpaceSettings=@n2
   ) {
    alphaInputType=Bool
    bClampLinearTranslationLimitToRefPose=true
    bOverrideWorldGravity=true
    overridePhysicsAsset=&/M_MED_InspireSpell_Alt_Cape_Physics
    simulationSpace=BaseBoneSpace
    alphaBoolBlend {blendOption=Linear}
    alphaScaleBiasClamp {bInterpResult=true}
    baseBoneRef {boneName=pelvis}
    componentAppliedLinearAccClamp=<1800 1800 1800>
    componentLinearAccScale=<0.4 0.15 0.15>
    componentLinearVelScale=<0.18 0.15 0.15>
    overrideWorldGravity {z=-980}
    simSpaceSettings {
     maxAngularAcceleration=250
     maxAngularVelocity=2
     maxLinearAcceleration=50
     maxLinearVelocity=120
     velocityScaleZ=0
     worldAlpha=1
     externalLinearDragV=<0.1 0.1>
    }
   }
   n7 = Component To Local(ComponentPose=@n6)
   n8 = Get bIsInFrontEnd()
   n9 = RigidBody!(
    ComponentPose=Local To Component(LocalPose=Blend Poses by bool(BlendPose_0=@n5, BlendPose_1=@n7, bActiveValue=@n8) {blendType=Linear})
    bAlphaBoolEnabled=Get Enable_Strap_RBAN() -> (Enable_Strap_RBAN:bool)
    OverrideWorldGravity=@n3
    ExternalForce
   ) -> (Pose) {
    alphaInputType=Bool
    bClampLinearTranslationLimitToRefPose=true
    bOverrideWorldGravity=true
    overridePhysicsAsset=&/M_MED_InspireSpell_Alt_SepStrap_Cape_Physics
    simulationSpace=BaseBoneSpace
    alphaBoolBlend {blendInTime=0.5, blendOption=Linear, blendOutTime=0.5}
    baseBoneRef {boneName=pelvis}
    componentAppliedLinearAccClamp=<3000 3000 3000>
    componentLinearAccScale=<0.15 0.15 0.15>
    componentLinearVelScale=<0.15 0.15 0.15>
    overrideWorldGravity {z=-980}
    simSpaceSettings {
     maxAngularAcceleration=500
     maxAngularVelocity=3
     maxLinearAcceleration=150
     maxLinearVelocity=250
     velocityScaleZ=0
     worldAlpha=1
     externalLinearDragV=<0.1 0.1>
    }
   }
   n10 = Get Gravity Override()
   n11 = RigidBody!(ComponentPose=@n9, Alpha:float, OverrideWorldGravity=@n10, ExternalForce) -> (Pose) {
    bClampLinearTranslationLimitToRefPose=true
    bOverrideWorldGravity=true
    overridePhysicsAsset=&/M_MED_InspireSpell_Alt_Shorts_Physics
    simulationSpace=BaseBoneSpace
    alphaBoolBlend {blendInTime=1, blendOption=Linear, blendOutTime=1}
    baseBoneRef {boneName=thigh_l}
    overrideWorldGravity {z=-950}
    simSpaceSettings {maxAngularVelocity=30, maxLinearAcceleration=250, maxLinearVelocity=500, velocityScaleZ=0, worldAlpha=0.832}
   }
   n12 = RigidBody(ComponentPose=@n11, OverrideWorldGravity=@n10) {
    bClampLinearTranslationLimitToRefPose=true
    bOverrideWorldGravity=true
    overridePhysicsAsset=&/M_MED_InspireSpell_Alt_Shorts_02_Physis
    simulationSpace=BaseBoneSpace
    alphaBoolBlend {blendInTime=1, blendOption=Linear, blendOutTime=1}
    baseBoneRef {boneName=thigh_r}
    overrideWorldGravity {z=-950}
    simSpaceSettings {maxAngularVelocity=30, maxLinearAcceleration=250, maxLinearVelocity=500, velocityScaleZ=0, worldAlpha=0.832}
   }
   n13 = RigidBody!(ComponentPose=@n12, bAlphaBoolEnabled:bool=Get Enable_Cape_RBAN(), OverrideWorldGravity=@n10, ExternalForce) -> (Pose) {
    alphaInputType=Bool
    bClampLinearTranslationLimitToRefPose=true
    bOverrideWorldGravity=true
    overridePhysicsAsset=&/M_MED_InspireSpell_Alt_Hood_Physics
    simulationSpace=BaseBoneSpace
    alphaBoolBlend {blendInTime=1, blendOption=Linear, blendOutTime=1}
    alphaScaleBiasClamp {bInterpResult=true}
    baseBoneRef {boneName=spine_05}
    componentAppliedLinearAccClamp=<3000 3000 3000>
    overrideWorldGravity {z=-950}
    simSpaceSettings {maxAngularAcceleration=2.5, maxAngularVelocity=200, maxLinearAcceleration=300, maxLinearVelocity=300, velocityScaleZ=0.288, worldAlpha=0.728}
   }
   n14 = Control Rig!(
    Source=Component To Local(ComponentPose=@n13)
    Alpha
    Is SkyFloating:bool=Get bIsSkydiveFloating()
    Is Parachute:bool=Get bIsParachuteOpened()
    Is Crouch:bool=Get bIsCrouching()
   ) -> (Pose) {
    alphaBoolBlend {blendOption=Linear}
    controlRigAssetReference {blueprintRigClass=&/M_MED_InspireSpell_Alt_CtrlRig.M_MED_InspireSpell_Alt_CtrlRig_C}
   }
   Output Pose(Result=@n14) {layerGroup=None}
   n15 = Get bIsSkyDiving()
   n16 = "Transform (Modify) Bone - Bone: dyn_cape_b_r_base"!(
    ComponentPose=Local To Component(LocalPose=Copy Pose From Mesh() {bUseAttachedParent=true})
    bAlphaBoolEnabled:bool=@n15
    Translation="10,0,0"
    Rotation="0,0,0"
    Scale="1,1,1"
   ) -> (Pose) {
    alphaInputType=Bool
    rotationMode=BMM_Additive
    rotationSpace=BCS_BoneSpace
    translationMode=BMM_Additive
    translationSpace=BCS_BoneSpace
    alphaBoolBlend {blendInTime=0.5, blendOption=Linear, blendOutTime=0.5}
    boneToModify {boneName=dyn_cape_b_r_base}
    translation {x=10}
   }
   n17 = "Transform (Modify) Bone - Bone: dyn_cape_b_l_base"!(ComponentPose=@n16, bAlphaBoolEnabled:bool=@n15, Translation="10,0,0", Rotation="0,0,0", Scale="1,1,1") -> (Pose) {
    alphaInputType=Bool
    rotationMode=BMM_Additive
    rotationSpace=BCS_BoneSpace
    translationMode=BMM_Additive
    translationSpace=BCS_BoneSpace
    alphaBoolBlend {blendInTime=0.5, blendOption=Linear, blendOutTime=0.5}
    boneToModify {boneName=dyn_cape_b_l_base}
    translation {x=10}
   }
   n18 = "Transform (Modify) Bone - Bone: dyn_cape_f_l_base"!(ComponentPose=@n17, bAlphaBoolEnabled:bool=@n15, Translation="10,0,0", Rotation="0,50,0", Scale="1,1,1") -> (Pose) {
    alphaInputType=Bool
    rotationMode=BMM_Additive
    rotationSpace=BCS_BoneSpace
    translationMode=BMM_Additive
    translationSpace=BCS_BoneSpace
    alphaBoolBlend {blendInTime=0.5, blendOption=Linear, blendOutTime=0.5}
    boneToModify {boneName=dyn_cape_f_l_base}
    rotation {yaw=50}
    translation {x=10}
   }
   n19 = "Transform (Modify) Bone - Bone: dyn_cape_f_r_base"!(ComponentPose=@n18, bAlphaBoolEnabled:bool=@n15, Translation="10,0,0", Rotation="0,-40,0", Scale="1,1,1") -> (Pose) {
    alphaInputType=Bool
    rotationMode=BMM_Additive
    rotationSpace=BCS_BoneSpace
    translationMode=BMM_Additive
    translationSpace=BCS_BoneSpace
    alphaBoolBlend {blendInTime=0.5, blendOption=Linear, blendOutTime=0.5}
    boneToModify {boneName=dyn_cape_f_r_base}
    rotation {yaw=-40}
    translation {x=10}
   }
   n20 = "Transform (Modify) Bone - Bone: dyn_cape_s_l_base"!(ComponentPose=@n19, bAlphaBoolEnabled:bool=@n15, Translation="10,0,0", Rotation="0,0,0", Scale="1,1,1") -> (Pose) {
    alphaInputType=Bool
    rotationMode=BMM_Additive
    rotationSpace=BCS_BoneSpace
    translationMode=BMM_Additive
    translationSpace=BCS_BoneSpace
    alphaBoolBlend {blendInTime=0.5, blendOption=Linear, blendOutTime=0.5}
    boneToModify {boneName=dyn_cape_s_l_base}
    translation {x=10}
   }
   n21 = "Transform (Modify) Bone - Bone: dyn_cape_s_r_base"!(ComponentPose=@n20, bAlphaBoolEnabled:bool=@n15, Translation="10,0,0", Rotation="0,0,0", Scale="1,1,1") -> (Pose) {
    alphaInputType=Bool
    rotationMode=BMM_Additive
    rotationSpace=BCS_BoneSpace
    translationMode=BMM_Additive
    translationSpace=BCS_BoneSpace
    alphaBoolBlend {blendInTime=0.5, blendOption=Linear, blendOutTime=0.5}
    boneToModify {boneName=dyn_cape_s_r_base}
    translation {x=10}
   }
   n22 = Get bIsKneeSliding()
   n23 = "Transform (Modify) Bone - Bone: dyn_cape_b_r_base"(ComponentPose=@n21, bAlphaBoolEnabled=@n22, Translation="-3,0,5", Rotation="90,0,0", Scale="1,1,1") {
    alphaInputType=Bool
    rotationMode=BMM_Additive
    rotationSpace=BCS_BoneSpace
    translationMode=BMM_Additive
    translationSpace=BCS_BoneSpace
    alphaBoolBlend {blendInTime=0.5, blendOption=Linear, blendOutTime=0.5}
    boneToModify {boneName=dyn_cape_b_r_base}
    rotation {pitch=90}
    translation {x=-3, z=5}
   }
   n24 = "Transform (Modify) Bone - Bone: dyn_cape_b_l_base"(ComponentPose=@n23, bAlphaBoolEnabled=@n22, Translation="-3,0,5", Rotation="90,0,0", Scale="1,1,1") {
    alphaInputType=Bool
    rotationMode=BMM_Additive
    rotationSpace=BCS_BoneSpace
    translationMode=BMM_Additive
    translationSpace=BCS_BoneSpace
    alphaBoolBlend {blendInTime=0.5, blendOption=Linear, blendOutTime=0.5}
    boneToModify {boneName=dyn_cape_b_l_base}
    rotation {pitch=90}
    translation {x=-3, z=5}
   }
   n25 = Control Rig!(
    Source=Component To Local(ComponentPose=@n24)
    Alpha
    Is enabled rigid body:bool=Get bRigidBodyEnabled()
    Is FrontEnd:bool=Get bIsInFrontEnd()
   ) -> (Pose) {
    alphaBoolBlend {blendInTime=0.5, blendOption=Linear, blendOutTime=0.5}
    controlRigAssetReference {blueprintRigClass=&/M_MED_InspireSpell_Alt_Movement_CtrlRig.M_MED_InspireSpell_Alt_Movement_CtrlRig_C}
   }
   Save cached pose 'Init Transforms'(Pose=@n25) {cacheName=Init Transforms}
   Get SepStrap() -> (SepStrap:bool)
   Get Enable_Strap_RBAN()
   Get Strap() -> (Strap:bool)
  ]}
 EventGraph {nodes [
   n1 = Event BlueprintUpdateAnimation()
   > n2 = Sequence()
   n3 = Get bIsMoving2D()
   n4 = Get bIsMoving2D()
   n5 = OR Boolean!(
    A=AND Boolean!(A=Get bIsSliding(), B=Get bIsCrouching(), C:bool=Get bIsMoving2D()) -> (ReturnValue)
    B=AND Boolean!(A=Get bIsCrouching(), B=NOT Boolean(A=@n4)) -> (ReturnValue)
    C:bool=AND Boolean(A=Get bIsDBNO(), B=NOT Boolean(A=@n3))
    D:bool=Get bIsInFrontEnd()
   ) -> (ReturnValue)
   > Set Strap(Strap:bool=NOT Boolean(A=@n5)) -> (Output_Get:bool)
   > Set SepStrap(SepStrap:bool=@n5) -> (Output_Get:bool)
   > n6 = Sequence()
   n7 = Branch(execute=@n6.then_1, Condition=Get bIsInFrontEnd())
   > Set Sim Space Settings(Sim Space Settings:Sim Space Settings Structure=Get Sim Space Settings FrontEnd() -> (Sim Space Settings FrontEnd:Sim Space Settings Structure)) -> (Output_Get:Sim Space Settings Structure)
   Set Sim Space Settings(execute=@n7.else, Sim Space Settings=Get Sim Space Settings BR() -> (Sim Space Settings BR:Sim Space Settings Structure))
   n8 = Gravity Override State Function DEPRECATED(
    execute=@n6.then_0
    CharacterPartAnimInstance=Self-Reference()
    GameTimeInSeconds=GetGameTimeInSeconds()
    GravityOverrideParams=Get Gravity Override Params()
   )
   > Set Gravity Override(Gravity Override=@n8)
   Is Valid(exec=@n2.then_1, InputObject=Get playerPawn() -> (playerPawn:Fort Player Pawn Athena Object Reference))
   > Set WPO Globals Function(
    Material=Get matInstanceDynamic() -> (matInstanceDynamic:Material Instance Dynamic Object Reference)
    Overlay Material=Get overlayMID() -> (overlayMID:Material Instance Dynamic Object Reference)
    Skeletal Mesh=Get skelMesh() -> (skelMesh:Skeletal Mesh Component Object Reference)
    "Local Noise Transform  Location"="Get Local Noise Transform  Location"() -> ("Local Noise Transform  Location":Vector)
    Local Noise Offset=Get Local Noise Offset()
   )
   > Set WPO Regional Function!(
    execute
    Material:Material Instance Dynamic Object Reference=Get matInstanceDynamic()
    Overlay Material:Material Instance Dynamic Object Reference=Get overlayMID()
    Skeletal Mesh:Skeletal Mesh Component Object Reference=Get skelMesh()
    Velocity Transform:Name=thigh_l
    Velocity Interp Speed:double=40
    Collision Transform:Name=thigh_l
    Collision Transform Axis:EAxis Enum
    Invert Collision Transform Axis:bool=true
    Delta Time X:double=@n1.DeltaTimeX
    Previous Velocity Transform Location:"Vector (by ref)"=Get R1 Velocity Transform Location()
    Previous Velocity Transform Velocity:"Vector (by ref)"=Get R1 Previous Velocity Transform Velocity() -> (R1 Previous Velocity Transform Velocity:Vector)
    Region:int=1
   ) -> (then)
   > Set WPO Regional Function(
    Material=Get matInstanceDynamic()
    Overlay Material=Get overlayMID()
    Skeletal Mesh=Get skelMesh()
    Velocity Transform=thigh_r
    Velocity Interp Speed=40
    Collision Transform=thigh_r
    Invert Collision Transform Axis=false
    Delta Time X=@n1.DeltaTimeX
    Previous Velocity Transform Location=Get R2 Velocity Transform Location()
    Previous Velocity Transform Velocity=Get R2 Velocity Transform Velocity()
    Region=2
   )
   > n9 = Branch(Condition=Get bIsSkydiveFloating())
   > n10 = Set WPO Pho Wind Intensity(WPO Pho Wind Intensity=1)
   SetScalarParameterValue(execute=[@n11.then, @n10.then], self=Get matInstanceDynamic(), ParameterName=ClothSim_WindStrength, Value=Get WPO Pho Wind Intensity())
   n11 = Set WPO Pho Wind Intensity(execute=@n9.else, WPO Pho Wind Intensity=0)
   Event BlueprintInitializeAnimation()
   > Set SkelMesh(SkelMesh=GetOwnerSkeletalMeshComponent())
   n12 = TryGetPawnOwner()
   > Is Valid(InputObject=@n12)
   > n13 = Cast To FortPlayerPawnAthena(Object=@n12)
   > Set playerPawn(playerPawn:Fort Player Pawn Athena Object Reference=@n13) -> (Output_Get:Fort Player Pawn Athena Object Reference)
   Event OnCharacterCustomizationLoaded()
   > n14 = Sequence!(execute) -> (then_0, then_1, then_2:Exec)
   > n15 = Set skelMesh(skelMesh:Skeletal Mesh Component Object Reference=GetPartSkeletalMeshComponent()) -> (Output_Get:Skeletal Mesh Component Object Reference)
   > n16 = Get Dynamic Material Instance(SkeletalMeshComponent=@n15)
   > Set matInstanceDynamic(matInstanceDynamic:Material Instance Dynamic Object Reference=@n16) -> (Output_Get:Material Instance Dynamic Object Reference)
   > n17 = GetOrCreateDynamicOverlayMaterial(MeshComponent=@n15, MaterialName=None)
   > Set overlayMID(overlayMID:Material Instance Dynamic Object Reference=@n17) -> (Output_Get:Material Instance Dynamic Object Reference)
   n18 = Cast To FortPlayerPawn(execute=@n14.then_1, Object=GetOwningActor())
   > n19 = GetVariantChannelActiveIndexForItemActor!(
    execute
    Item:Athena Cosmetic Item Definition Object Reference=/BRCosmetics/Athena/Items/Cosmetics/Characters/Character_InspireSpell
    Actor:Actor Object Reference=@n18
    Channel:Gameplay Tag Structure=(TagName="Cosmetics.Variant.Channel.Parts")
    VariantList:Array of Gameplay Tag Structures=Make Array!("[0]"=(TagName="Cosmetics.Variant.Property.Stage3"), "[1]":Gameplay Tag Structure=(TagName="Cosmetics.Variant.Property.Stage5")) -> (Array)
    bErrorIfChannelMissing:bool=true
    bErrorIfTagNotInArray:bool=true
   ) -> (then, Index:int=0, bIndexIsValid:bool, bFoundChannel:bool)
   > n20 = Branch(Condition=@n19.bIndexIsValid)
   > Set Enable_Cape_RBAN(Enable_Cape_RBAN:bool=false) -> (Output_Get:bool)
   Set Enable_Cape_RBAN(execute=@n20.else, Enable_Cape_RBAN=true)
   n21 = Get Enable_Cape_RBAN()
   n22 = Get Strap()
   n23 = Get SepStrap()
   n24 = Branch(execute=@n14.then_2, Condition=OR Boolean!(A=AND Boolean(A=@n21, B=@n22), B=AND Boolean(A=@n21, B=@n23)) -> (ReturnValue))
   > Set Enable_Strap_RBAN(Enable_Strap_RBAN:bool=true) -> (Output_Get:bool)
   Set Enable_Strap_RBAN(execute=@n24.else, Enable_Strap_RBAN=false)
  ]}
]
variables [
 Crouch/Sliding {type=boolean}
 Enable_Cape_RBAN {type=boolean}
 Enable_Strap_RBAN {type=boolean}
 Gravity Override {type=Vector}
 Gravity Override Params {
  type=GravityOverrideParamsStruct
  default {
   front_End {gravityOverride=<0 0 -600>}
   skydive_DivingDown {jointName=root, gravityOverride=<0 600 1500>, windAmplitude=<50 300 300>, windFrequency=<1.5 12 15>, windOffset=<0 0 -500>}
   skydive_FloatingDown {gravityOverride=<0 -700 750>}
   skydive_GoingUp {gravityOverride=<0 -700 1500>, windAmplitude=<200 -500 0>, windOffset=<0 -600 0>}
   tactical_Sprint {gravityOverride=<0 -500 -450>, windOffset=<0 0 500>}
  }
 }
 Local Noise Offset {type=Vector}
 "Local Noise Transform  Location" {type=Vector}
 matInstanceDynamic {type=MaterialInstanceDynamic}
 overlayMID {type=MaterialInstanceDynamic}
 playerPawn {type=FortPlayerPawnAthena}
 R1 Previous Velocity Transform Velocity {type=Vector}
 R1 Velocity Transform Location {type=Vector}
 R2 Velocity Transform Location {type=Vector}
 R2 Velocity Transform Velocity {type=Vector}
 SepStrap {type=boolean}
 Sim Space Settings {
  type=SimSpaceSettings
  default {
   dampingAlpha=1
   maxAngularAcceleration=250
   maxAngularVelocity=2
   maxLinearAcceleration=50
   maxLinearVelocity=120
   velocityScaleZ=0
   worldAlpha=1
   externalLinearDragV=<0.1 0.1 0>
  }
 }
 Sim Space Settings BR {
  type=SimSpaceSettings
  default {
   dampingAlpha=1
   maxAngularAcceleration=120
   maxAngularVelocity=1.5
   maxLinearAcceleration=50
   maxLinearVelocity=130
   velocityScaleZ=0
   worldAlpha=1
   externalLinearDragV=<0.1 0.1 0>
  }
 }
 Sim Space Settings FrontEnd {
  type=SimSpaceSettings
  default {
   dampingAlpha=1
   maxAngularAcceleration=50
   maxAngularVelocity=1.5
   maxLinearAcceleration=50
   maxLinearVelocity=130
   velocityScaleZ=0
   worldAlpha=1
   externalLinearDragV=<0.1 0.18 0.1>
  }
 }
 skelMesh {type=SkeletalMeshComponent}
 Strap {type=boolean}
 WPO Pho Wind Intensity {type=number}
]
